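# Control Rig for the Stone Viper "Axe" character part, orienting passive snake chains attached to the head.
# Five chained AimItem nodes point the X axis of the b–f snake master bones at fixed offsets in spine_05 space, so they follow the upper spine rather than head rotation alone.
# Secondary axes are zero, so roll is unconstrained.
# Snake a, the "active" snake set and all child chain bones receive no direct edits.
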ControlRigBlueprint /BRCosmetics/Characters/Player/Female/Medium/Bodies/F_MED_Stone_Viper/Meshes/Parts/F_MED_Stone_Viper_Axe_CtrlRig
graphs [
 RigVMFunctionLibrary {links [], nodes []}
 RigVMModel {nodes [
   RigUnit_BeginExecution = &RigUnit_BeginExecution()
   /Script/ControlRig.RigUnit_AimItem(
    DebugSettings=(Scale=10,WorldOffset=(Rotation=(X=0,Y=0,Z=0,W=1),Translation=(X=0,Y=0,Z=0),Scale3D=(X=1,Y=1,Z=1)))
    Item=(Type=Bone,Name="dyn_snake_b_MSTR")
    Primary=(Weight=1,Axis=(X=1,Y=0,Z=0),Target=(X=0,Y=-20,Z=-15),Kind=Location,Space=(Type=Bone,Name="spine_05"))
    Secondary=(Weight=1,Axis=(X=0,Y=0,Z=0),Target=(X=0,Y=0,Z=0),Kind=Location,Space=(Type=Bone))
    Weight=1
    ExecutePin=@RigUnit_BeginExecution.ExecutePin
   ) {is=struct RigVMUnitNode}
   > &RigUnit_AimItem(
    DebugSettings=(Scale=10,WorldOffset=(Rotation=(X=0,Y=0,Z=0,W=1),Translation=(X=0,Y=0,Z=0),Scale3D=(X=1,Y=1,Z=1)))
    Item=(Type=Bone,Name="dyn_snake_c_MSTR")
    Primary=(Weight=1,Axis=(X=1,Y=0,Z=0),Target=(X=0,Y=-20,Z=0),Kind=Location,Space=(Type=Bone,Name="spine_05"))
    Secondary=(Weight=1,Axis=(X=0,Y=0,Z=0),Target=(X=0,Y=0,Z=0),Kind=Location,Space=(Type=Bone))
    Weight=1
   )
   > &RigUnit_AimItem(
    DebugSettings=(Scale=10,WorldOffset=(Rotation=(X=0,Y=0,Z=0,W=1),Translation=(X=0,Y=0,Z=0),Scale3D=(X=1,Y=1,Z=1)))
    Item=(Type=Bone,Name="dyn_snake_d_MSTR")
    Primary=(Weight=1,Axis=(X=1,Y=0,Z=0),Target=(X=0,Y=-15,Z=3),Kind=Location,Space=(Type=Bone,Name="spine_05"))
    Secondary=(Weight=1,Axis=(X=0,Y=0,Z=0),Target=(X=0,Y=0,Z=0),Kind=Location,Space=(Type=Bone))
    Weight=1
   )
   > &RigUnit_AimItem(
    DebugSettings=(Scale=10,WorldOffset=(Rotation=(X=0,Y=0,Z=0,W=1),Translation=(X=0,Y=0,Z=0),Scale3D=(X=1,Y=1,Z=1)))
    Item=(Type=Bone,Name="dyn_snake_e_MSTR")
    Primary=(Weight=1,Axis=(X=1,Y=0,Z=0),Target=(X=0,Y=-17,Z=7),Kind=Location,Space=(Type=Bone,Name="spine_05"))
    Secondary=(Weight=1,Axis=(X=0,Y=0,Z=0),Target=(X=0,Y=0,Z=0),Kind=Location,Space=(Type=Bone))
    Weight=1
   )
   > &RigUnit_AimItem(
    DebugSettings=(Scale=10,WorldOffset=(Rotation=(X=0,Y=0,Z=0,W=1),Translation=(X=0,Y=0,Z=0),Scale3D=(X=1,Y=1,Z=1)))
    Item=(Type=Bone,Name="dyn_snake_f_MSTR")
    Primary=(Weight=1,Axis=(X=1,Y=0,Z=0),Target=(X=0,Y=-8,Z=9),Kind=Location,Space=(Type=Bone,Name="spine_05"))
    Secondary=(Weight=1,Axis=(X=0,Y=0,Z=0),Target=(X=0,Y=0,Z=0),Kind=Location,Space=(Type=Bone))
    Weight=1
   )
  ]}
]
hierarchy [
 root {loc=<0 0 0>, rot=<0 0 0 1>}
 pelvis {parent=root, loc=<0 -0.5301 98.0571>, rot=<0.0127 0.707 -0.0127 -0.707>}
 spine_01 {parent=pelvis, loc=<6.1208 -0.9328 0>, rot=<0 0 0.034 -0.9994>}
 spine_02 {parent=spine_01, loc=<6.5996 -0.2095 0>, rot=<0 0 0.0238 -0.9997>}
 spine_03 {parent=spine_02, loc=<6.4914 -0.1237 0>, rot=<0 0 0.0149 -0.9999>}
 spine_04 {parent=spine_03, loc=<6.9598 0.0794 0>, rot=<0 0 -0.0132 -0.9999>}
 spine_05 {parent=spine_04, loc=<6.9987 0 0>, rot=<0 0 -0.0015 -1>}
 clavicle_l {parent=spine_05, loc=<5.4831 -0.1053 -3.2946>, rot=<0.0465 -0.7289 -0.009 -0.6829>}
 upperarm_l {parent=clavicle_l, loc=<9.8658 -1.2136 -0.4105>, rot=<0.0208 -0.486 -0.0242 -0.8734>}
 upperarm_twist_01_l {parent=upperarm_l, loc=<7.2135 -0.1134 0>, rot=<0 0 0 1>}
 lowerarm_l {parent=upperarm_l, loc=<26.465 -0.8345 0.259>, rot=<0.0036 0.0161 -0.2298 -0.9731>}
 upperarm_twist_02_l {parent=upperarm_l, loc=<19.7998 -0.1134 0>, rot=<0 0 0 1>}
 clavicle_r {parent=spine_05, loc=<5.4831 -0.1053 3.2946>, rot=<0.7289 0.0465 -0.6829 0.009>}
 upperarm_r {parent=clavicle_r, loc=<-9.8658 1.2136 0.4104>, rot=<0.0208 -0.486 -0.0242 -0.8734>}
 upperarm_twist_01_r {parent=upperarm_r, loc=<-7.2135 0.1134 0>, rot=<0 0 0 1>}
 lowerarm_r {parent=upperarm_r, loc=<-26.4651 0.8345 -0.259>, rot=<0.0036 0.0161 -0.2298 -0.9731>}
 upperarm_twist_02_r {parent=upperarm_r, loc=<-19.7998 0.1134 0>, rot=<0 0 0 1>}
 neck_01 {parent=spine_05, loc=<11.3138 0.1529 0>, rot=<0 0 -0.2003 -0.9797>}
 neck_02 {parent=neck_01, loc=<4.5627 0 0>, rot=<0 0 0.0504 -0.9987>}
 head {parent=neck_02, loc=<5.3914 0 0>, rot=<0 0 0.115 -0.9934>}
 dyn_active_MSTR {parent=head, loc=<6.6193 0 0>, rot=<0 0 0 -1>}
 dyn_snake_active_a_MSTR {parent=dyn_active_MSTR, loc=<-0.8294 4.9891 -7.341>, rot=<0.004 -0.9779 -0.0009 -0.2092>}
 dyn_snake_active_a_01 {parent=dyn_snake_active_a_MSTR, loc=<0.5233 0 0>, rot=<-0.3226 0.0934 0.1857 -0.9234>}
 dyn_snake_active_a_02 {parent=dyn_snake_active_a_01, loc=<3.4433 0 0>, rot=<-0.0899 0.1617 0.0902 -0.9786>}
 dyn_snake_active_a_03 {parent=dyn_snake_active_a_02, loc=<3.8388 0 0>, rot=<-0.0899 0.2269 0.2223 -0.9439>}
 dyn_snake_active_a_04 {parent=dyn_snake_active_a_03, loc=<4.053 0 0>, rot=<0.0006 0.142 -0.1484 -0.9787>}
 dyn_snake_active_a_05 {parent=dyn_snake_active_a_04, loc=<4.0876 0 0>, rot=<-0.0712 0.1498 -0.2342 -0.9579>}
 dyn_snake_active_a_06 {parent=dyn_snake_active_a_05, loc=<4.4 0 0>, rot=<-0.0752 -0.1277 -0.2282 -0.9623>}
 dyn_snake_active_a_07 {parent=dyn_snake_active_a_06, loc=<4.0036 0 0>, rot=<0.0071 -0.1447 0.1852 -0.972>}
 dyn_snake_active_a_08 {parent=dyn_snake_active_a_07, loc=<3.5699 0 0>, rot=<0.0059 0.0991 0.5591 -0.8231>}
 dyn_snake_active_a_09 {parent=dyn_snake_active_a_08, loc=<2.9362 0 0>, rot=<0.0618 0.0613 0.3128 -0.9458>}
 snake_a_active_jaw {parent=dyn_snake_active_a_09, loc=<2.546 -0.6386 0.1573>, rot=<0.1512 -0.0524 0.0128 -0.987>}
 dyn_snake_active_b_MSTR {parent=dyn_active_MSTR, loc=<-6.0133 -5.6586 -5.7541>, rot=<0.0982 -0.8856 0.4029 -0.2091>}
 dyn_snake_active_b_01 {parent=dyn_snake_active_b_MSTR, loc=<0.7607 0 0>, rot=<-0.0834 0.3301 0.0857 -0.9363>}
 dyn_snake_active_b_02 {parent=dyn_snake_active_b_01, loc=<3.7617 0 0>, rot=<0.0474 0.1895 -0.1953 -0.9611>}
 dyn_snake_active_b_03 {parent=dyn_snake_active_b_02, loc=<4.0184 0 0>, rot=<0.0261 -0.2268 -0.093 -0.9691>}
 dyn_snake_active_b_04 {parent=dyn_snake_active_b_03, loc=<4.782 0 0>, rot=<0.0628 -0.3564 -0.0531 -0.9307>}
 dyn_snake_active_b_05 {parent=dyn_snake_active_b_04, loc=<4.7721 0 0>, rot=<0.5036 0.0936 -0.2426 -0.8239>}
 dyn_snake_active_b_06 {parent=dyn_snake_active_b_05, loc=<3.9781 0 0>, rot=<0.0704 0.4831 -0.1293 -0.8631>}
 dyn_snake_active_b_07 {parent=dyn_snake_active_b_06, loc=<3.9256 0 0>, rot=<-0.0056 0.0711 -0.0666 -0.9952>}
 dyn_snake_active_b_08 {parent=dyn_snake_active_b_07, loc=<4.0689 0 0>, rot=<0.0201 -0.2687 -0.165 -0.9488>}
 dyn_snake_active_b_09 {parent=dyn_snake_active_b_08, loc=<3.1159 0 0>, rot=<-0.2262 -0.2835 -0.015 -0.9318>}
 snake_active_b_jaw {parent=dyn_snake_active_b_09, loc=<2.4568 0.5291 -0.0083>, rot=<0.0218 -0.0673 0.0092 -0.9975>}
 dyn_snake_active_c_MSTR {parent=dyn_active_MSTR, loc=<-4.3773 -9.6365 -4.3201>, rot=<0.1983 -0.6787 0.6787 -0.1983>}
 dyn_snake_active_c_01 {parent=dyn_snake_active_c_MSTR, loc=<0.5187 0 0>, rot=<0.0843 0.2524 -0.0899 -0.9597>}
 dyn_snake_active_c_02 {parent=dyn_snake_active_c_01, loc=<4.27 0 0>, rot=<0.2174 -0.24 -0.2324 -0.9171>}
 dyn_snake_active_c_03 {parent=dyn_snake_active_c_02, loc=<4.008 0 0>, rot=<0.7407 -0.0508 -0.3774 -0.5535>}
 dyn_snake_active_c_04 {parent=dyn_snake_active_c_03, loc=<4.9131 0 0>, rot=<0.0578 0.3592 -0.0537 -0.9299>}
 dyn_snake_active_c_05 {parent=dyn_snake_active_c_04, loc=<4.7932 0 0>, rot=<-0.0936 0.175 0.3442 -0.9177>}
 dyn_snake_active_c_06 {parent=dyn_snake_active_c_05, loc=<4.918 0 0>, rot=<0.041 0.1717 0.3734 -0.9107>}
 dyn_snake_active_c_07 {parent=dyn_snake_active_c_06, loc=<4.0664 0 0>, rot=<0.0213 0.051 0.0589 -0.9967>}
 snake_active_c_jaw {parent=dyn_snake_active_c_07, loc=<1.864 0.0021 0.5784>, rot=<0.0002 -0.0881 0.0006 -0.9961>}
 dyn_snake_active_d_MSTR {parent=dyn_active_MSTR, loc=<-3.7634 -9.2586 -0.0438>, rot=<0.14 -0.8472 0.4387 -0.265>}
 dyn_snake_active_d_01 {parent=dyn_snake_active_d_MSTR, loc=<0.576 0 0>, rot=<-0.1019 0.1918 0.1043 -0.9705>}
 dyn_snake_active_d_02 {parent=dyn_snake_active_d_01, loc=<5.2836 0 0>, rot=<0.0375 0.2558 -0.1138 -0.9593>}
 dyn_snake_active_d_03 {parent=dyn_snake_active_d_02, loc=<5.651 0 0>, rot=<0.0363 -0.0657 -0.2862 -0.9552>}
 dyn_snake_active_d_04 {parent=dyn_snake_active_d_03, loc=<5.5394 0 0>, rot=<0.0889 -0.0741 -0.3173 -0.9412>}
 dyn_snake_active_d_05 {parent=dyn_snake_active_d_04, loc=<5.8605 0 0>, rot=<0.0257 0.247 -0.2798 -0.9274>}
 dyn_snake_active_d_06 {parent=dyn_snake_active_d_05, loc=<5.3923 0 0>, rot=<-0.1147 0.1843 -0.3006 -0.9287>}
 dyn_snake_active_d_07 {parent=dyn_snake_active_d_06, loc=<4.8399 0 0>, rot=<-0.0767 -0.1396 -0.1727 -0.972>}
 dyn_snake_active_d_08 {parent=dyn_snake_active_d_07, loc=<5.5548 0 0>, rot=<0.0022 -0.1741 0.022 -0.9845>}
 dyn_snake_active_d_09 {parent=dyn_snake_active_d_08, loc=<5.205 0 0>, rot=<0.1235 -0.0446 0.0415 -0.9905>}
 snake_d_active_jaw {parent=dyn_snake_active_d_09, loc=<2.2903 0.5492 0.4645>, rot=<0.0128 -0.0768 0.0309 -0.9965>}
 dyn_snake_active_e_MSTR {parent=dyn_active_MSTR, loc=<-1.8534 -6.9819 4.7797>, rot=<0.1327 -0.6945 0.6945 -0.1327>}
 dyn_snake_active_e_01 {parent=dyn_snake_active_e_MSTR, loc=<0.4362 0 0>, rot=<0.3798 0.016 -0.1547 -0.9119>}
 dyn_snake_active_e_02 {parent=dyn_snake_active_e_01, loc=<5.5453 0 0>, rot=<0.2339 0.1284 -0.1397 -0.9536>}
 dyn_snake_active_e_03 {parent=dyn_snake_active_e_02, loc=<6.1495 0 0>, rot=<-0.1749 0.0757 0.1631 -0.968>}
 dyn_snake_active_e_04 {parent=dyn_snake_active_e_03, loc=<6.7661 0 0>, rot=<0.0044 0.4167 -0.0138 -0.9089>}
 dyn_snake_active_e_05 {parent=dyn_snake_active_e_04, loc=<6.0965 0 0>, rot=<-0.057 0.1339 -0.2186 -0.9649>}
 dyn_snake_active_e_06 {parent=dyn_snake_active_e_05, loc=<5.1827 0 0>, rot=<-0.0046 -0.3423 -0.1112 -0.933>}
 dyn_snake_active_e_07 {parent=dyn_snake_active_e_06, loc=<4.7896 0 0>, rot=<-0.0322 -0.1946 0.0581 -0.9786>}
 dyn_snake_active_e_08 {parent=dyn_snake_active_e_07, loc=<4.8198 0 0>, rot=<-0.1195 0.3103 0.3103 -0.8906>}
 dyn_snake_active_e_09 {parent=dyn_snake_active_e_08, loc=<4.2051 0 0>, rot=<0.0029 0.1082 0.0377 -0.9934>}
 snake_e_active_jaw {parent=dyn_snake_active_e_09, loc=<1.5197 -0.2596 0.2764>, rot=<-0.0691 -0.0216 0.0009 -0.9974>}
 dyn_snake_active_f_MSTR {parent=dyn_active_MSTR, loc=<-7.7167 -5.0898 6.441>, rot=<0.0029 -0.7071 0.7071 -0.0029>}
 dyn_snake_active_f_01 {parent=dyn_snake_active_f_MSTR, loc=<0.4336 0 0>, rot=<0.0745 0.3543 -0.0283 -0.9317>}
 dyn_snake_active_f_02 {parent=dyn_snake_active_f_01, loc=<6.4284 0 0>, rot=<0.0766 0.2351 -0.111 -0.9626>}
 dyn_snake_active_f_03 {parent=dyn_snake_active_f_02, loc=<8.2494 0 0>, rot=<0.0273 0.2016 -0.1708 -0.9641>}
 dyn_snake_active_f_04 {parent=dyn_snake_active_f_03, loc=<7.8834 0 0>, rot=<-0.0388 0.2258 -0.1362 -0.9638>}
 dyn_snake_active_f_05 {parent=dyn_snake_active_f_04, loc=<6.8891 0 0>, rot=<-0.1294 0.012 -0.226 -0.9654>}
 dyn_snake_active_f_06 {parent=dyn_snake_active_f_05, loc=<6.9536 0 0>, rot=<-0.0387 -0.2614 -0.1426 -0.9538>}
 dyn_snake_active_f_07 {parent=dyn_snake_active_f_06, loc=<6.1342 0 0>, rot=<0.0381 -0.2062 -0.1754 -0.9619>}
 dyn_snake_active_f_08 {parent=dyn_snake_active_f_07, loc=<6.3973 0 0>, rot=<0.181 -0.0343 -0.363 -0.9134>}
 dyn_snake_active_f_09 {parent=dyn_snake_active_f_08, loc=<5.521 0 0>, rot=<0.0202 0.0359 -0.0403 -0.9983>}
 snake_f_active_jaw {parent=dyn_snake_active_f_09, loc=<2.5316 0.6912 0.379>, rot=<-0.2245 -0.0679 0.0029 -0.9721>}
 dyn_passive_MSTR {parent=head, loc=<6.6193 0 0>, rot=<0 0 0 -1>}
 dyn_snake_a_MSTR {parent=dyn_passive_MSTR, loc=<0.3845 5.4573 -6.7781>, rot=<0.0039 -0.951 -0.0013 -0.309>}
 dyn_snake_a_01 {parent=dyn_snake_a_MSTR, loc=<0.7272 0 0>, rot=<-0.489 -0.1244 0.2538 -0.8252>}
 dyn_snake_a_02 {parent=dyn_snake_a_01, loc=<3.969 0 0>, rot=<-0.3161 0.0897 0.143 -0.9336>}
 dyn_snake_a_03 {parent=dyn_snake_a_02, loc=<4.4361 0 0>, rot=<-0.0293 0.1274 0.0222 -0.9912>}
 dyn_snake_a_04 {parent=dyn_snake_a_03, loc=<4.6156 0 0>, rot=<0.0461 0.0418 -0.0492 -0.9969>}
 dyn_snake_a_05 {parent=dyn_snake_a_04, loc=<4.4825 0 0>, rot=<0.2052 0.019 -0.2482 -0.9465>}
 dyn_snake_a_06 {parent=dyn_snake_a_05, loc=<4.9431 0 0>, rot=<0.085 -0.0404 -0.0987 -0.9907>}
 dyn_snake_a_07 {parent=dyn_snake_a_06, loc=<4.8551 0 0>, rot=<0.1953 -0.2096 -0.134 -0.9487>}
 dyn_snake_a_08 {parent=dyn_snake_a_07, loc=<4.6864 0 0>, rot=<-0.1286 -0.2259 -0.0914 -0.9613>}
 dyn_snake_a_09 {parent=dyn_snake_a_08, loc=<4.4323 0 0>, rot=<-0.1493 -0.1335 0.1174 -0.9727>}
 snake_a_joint_jaw {parent=dyn_snake_a_09, loc=<1.7686 -0.0734 -0.1802>, rot=<-0.0575 -0.0954 0.0391 -0.993>}
 dyn_snake_b_MSTR {parent=dyn_passive_MSTR, loc=<-2.9568 -5.1056 -6.2688>, rot=<0.1525 -0.8489 0.3828 -0.3311>}
 dyn_snake_b_01 {parent=dyn_snake_b_MSTR, loc=<0.8691 0 0>, rot=<-0.0657 -0.2978 0.1537 -0.9399>}
 dyn_snake_b_02 {parent=dyn_snake_b_01, loc=<2.5381 0 0>, rot=<-0.0609 -0.0284 -0.0572 -0.9961>}
 dyn_snake_b_03 {parent=dyn_snake_b_02, loc=<4.1333 0 0>, rot=<-0.1501 0.1032 -0.0126 -0.9832>}
 dyn_snake_b_04 {parent=dyn_snake_b_03, loc=<3.422 0 0>, rot=<0.041 0.1346 -0.021 -0.9898>}
 dyn_snake_b_05 {parent=dyn_snake_b_04, loc=<3.6376 0 0>, rot=<-0.057 -0.0499 0.0355 -0.9965>}
 dyn_snake_b_06 {parent=dyn_snake_b_05, loc=<4.437 0 0>, rot=<0.0919 -0.2524 0.0358 -0.9626>}
 dyn_snake_b_07 {parent=dyn_snake_b_06, loc=<4.4057 0 0>, rot=<-0.0535 -0.1064 -0.025 -0.9926>}
 dyn_snake_b_08 {parent=dyn_snake_b_07, loc=<3.6419 0 0>, rot=<-0.0203 0.0902 -0.0971 -0.991>}
 dyn_snake_b_09 {parent=dyn_snake_b_08, loc=<3.6427 0 0>, rot=<-0.0041 -0.0242 -0.0562 -0.9981>}
 snake_b_joint_jaw {parent=dyn_snake_b_09, loc=<1.3687 -0.0301 -0.1197>, rot=<0.0067 -0.0591 -0.0026 -0.9982>}
 dyn_snake_c_MSTR {parent=dyn_passive_MSTR, loc=<-1.1074 -7.4126 -2.9671>, rot=<0.1858 -0.6823 0.6823 -0.1858>}
 dyn_snake_c_01 {parent=dyn_snake_c_MSTR, loc=<1.0574 0 0>, rot=<0.023 -0.2385 0.0094 -0.9708>}
 dyn_snake_c_02 {parent=dyn_snake_c_01, loc=<3.4298 0 0>, rot=<-0.0455 -0.0458 -0.0042 -0.9979>}
 dyn_snake_c_03 {parent=dyn_snake_c_02, loc=<4.3502 0 0>, rot=<0.0458 0.0562 -0.0576 -0.9957>}
 dyn_snake_c_04 {parent=dyn_snake_c_03, loc=<4.3006 0 0>, rot=<-0.0443 0.1279 -0.0091 -0.9908>}
 dyn_snake_c_05 {parent=dyn_snake_c_04, loc=<4.5335 0 0>, rot=<-0.0542 0.0015 0.0837 -0.995>}
 dyn_snake_c_06 {parent=dyn_snake_c_05, loc=<4.3502 0 0>, rot=<0.1494 -0.0975 0.1147 -0.9772>}
 dyn_snake_c_07 {parent=dyn_snake_c_06, loc=<4.0618 0 0>, rot=<-0.007 -0.1279 -0.0621 -0.9898>}
 dyn_snake_c_08 {parent=dyn_snake_c_07, loc=<3.8967 0 0>, rot=<0.0076 0.0162 -0.0807 -0.9966>}
 dyn_snake_c_09 {parent=dyn_snake_c_08, loc=<4.2679 0 0>, rot=<-0.0995 -0.0164 -0.1052 -0.9893>}
 snake_c_joint_jaw {parent=dyn_snake_c_09, loc=<1.5184 0.1385 -0.2438>, rot=<0.0974 -0.0211 0.0214 -0.9948>}
 dyn_snake_d_MSTR {parent=dyn_passive_MSTR, loc=<0.2666 -7.8325 1.1901>, rot=<0.1409 -0.6929 0.6929 -0.1409>}
 dyn_snake_d_01 {parent=dyn_snake_d_MSTR, loc=<0.9035 0 0>, rot=<0.0499 -0.108 -0.0146 -0.9928>}
 dyn_snake_d_02 {parent=dyn_snake_d_01, loc=<3.7389 0 0>, rot=<-0.0117 -0.0923 -0.0282 -0.9953>}
 dyn_snake_d_03 {parent=dyn_snake_d_02, loc=<4.2926 0 0>, rot=<-0.0182 -0.0039 -0.0258 -0.9995>}
 dyn_snake_d_04 {parent=dyn_snake_d_03, loc=<4.9489 0 0>, rot=<0.0876 0.1157 0.0458 -0.9884>}
 dyn_snake_d_05 {parent=dyn_snake_d_04, loc=<4.6259 0 0>, rot=<-0.0741 0.0926 0.0738 -0.9902>}
 dyn_snake_d_06 {parent=dyn_snake_d_05, loc=<4.7131 0 0>, rot=<-0.0104 -0.1113 0.0078 -0.9937>}
 dyn_snake_d_07 {parent=dyn_snake_d_06, loc=<5.0908 0 0>, rot=<0.0273 -0.172 -0.0569 -0.9831>}
 dyn_snake_d_08 {parent=dyn_snake_d_07, loc=<5.0087 0 0>, rot=<-0.0259 0.0401 -0.0343 -0.9983>}
 dyn_snake_d_09 {parent=dyn_snake_d_08, loc=<5.0745 0 0>, rot=<-0.0238 -0.0365 -0.1197 -0.9919>}
 snake_d_joint_jaw {parent=dyn_snake_d_09, loc=<1.4952 0.291 -0.0891>, rot=<0.0698 -0.0597 0.0157 -0.9956>}
 dyn_snake_e_MSTR {parent=dyn_passive_MSTR, loc=<-0.1175 -6.3433 4.6508>, rot=<0.1645 -0.6877 0.6877 -0.1645>}
 dyn_snake_e_01 {parent=dyn_snake_e_MSTR, loc=<0.8363 0 0>, rot=<-0.0223 -0.0871 0.0087 -0.9959>}
 dyn_snake_e_02 {parent=dyn_snake_e_01, loc=<3.5136 0 0>, rot=<-0.0309 -0.1832 -0.0713 -0.98>}
 dyn_snake_e_03 {parent=dyn_snake_e_02, loc=<4.6268 0 0>, rot=<-0.0507 0.0331 -0.0116 -0.9981>}
 dyn_snake_e_04 {parent=dyn_snake_e_03, loc=<5.1048 0 0>, rot=<0.0993 0.2075 0.0654 -0.971>}
 dyn_snake_e_05 {parent=dyn_snake_e_04, loc=<4.7661 0 0>, rot=<-0.0923 0.0766 0.0476 -0.9916>}
 dyn_snake_e_06 {parent=dyn_snake_e_05, loc=<4.9165 0 0>, rot=<0.1224 -0.2749 -0.0563 -0.952>}
 dyn_snake_e_07 {parent=dyn_snake_e_06, loc=<4.9571 0 0>, rot=<-0.0403 -0.153 -0.0829 -0.9839>}
 dyn_snake_e_08 {parent=dyn_snake_e_07, loc=<4.8475 0 0>, rot=<0.0241 0.0922 0.1502 -0.9841>}
 dyn_snake_e_09 {parent=dyn_snake_e_08, loc=<4.4917 0 0>, rot=<-0.0062 -0.0295 0.07 -0.9971>}
 snake_e_joint_jaw {parent=dyn_snake_e_09, loc=<1.3941 -0.1268 -0.0695>, rot=<0.0023 -0.0647 0.0116 -0.9978>}
 dyn_snake_f_MSTR {parent=dyn_passive_MSTR, loc=<-0.7807 -4.2966 6.6699>, rot=<0.0667 -0.7039 0.7039 -0.0667>}
 dyn_snake_f_01 {parent=dyn_snake_f_MSTR, loc=<1.1075 -0.2215 0.1674>, rot=<-0.2039 0.0541 0.0456 -0.9764>}
 dyn_snake_f_02 {parent=dyn_snake_f_01, loc=<3.4248 0.0994 -0.174>, rot=<0.134 -0.1595 -0.1062 -0.9723>}
 dyn_snake_f_03 {parent=dyn_snake_f_02, loc=<4.1631 0 0>, rot=<-0.0043 -0.0198 0.0071 -0.9998>}
 dyn_snake_f_04 {parent=dyn_snake_f_03, loc=<4.0531 0 0>, rot=<-0.0378 0.2698 0.0938 -0.9576>}
 dyn_snake_f_05 {parent=dyn_snake_f_04, loc=<4.2882 0 0>, rot=<0.0499 0.11 0.0181 -0.9925>}
 dyn_snake_f_06 {parent=dyn_snake_f_05, loc=<4.7658 0 0>, rot=<0.0392 -0.2305 -0.0165 -0.9721>}
 dyn_snake_f_07 {parent=dyn_snake_f_06, loc=<4.1411 0 0>, rot=<0.048 -0.2461 -0.0913 -0.9638>}
 dyn_snake_f_08 {parent=dyn_snake_f_07, loc=<4.5005 0 0>, rot=<0.0932 -0.0451 -0.0756 -0.9917>}
 dyn_snake_f_09 {parent=dyn_snake_f_08, loc=<4.2382 0 0>, rot=<0.0296 0.0572 -0.0024 -0.9979>}
 snake_f_joint_jaw {parent=dyn_snake_f_09, loc=<1.569 0.3466 0.4261>, rot=<-0.134 -0.0895 0.0002 -0.9869>}
]
