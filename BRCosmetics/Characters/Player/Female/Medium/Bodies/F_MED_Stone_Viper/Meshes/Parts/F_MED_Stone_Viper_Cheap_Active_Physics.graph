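# Low-cost physics asset for the active snake parts of the F_MED_Stone_Viper player cosmetic body, using only capsule bodies.
# Spine and clavicle capsules form the upper-torso base, with six small snake master bodies joined to spine_05 by angularly free constraints.
# Only snakes b, d and f carry sparse chains of bodies on selected bones, with tight twist and per-chain swing limits; a, c and e have just their master body.
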PhysicsAsset /BRCosmetics/Characters/Player/Female/Medium/Bodies/F_MED_Stone_Viper/Meshes/Parts/F_MED_Stone_Viper_Cheap_Active_Physics
bodies {
 clavicle_l [{shapeType=Capsule, length=8.3017, radius=5.755, center=<5 0 0>, rotation=<rot 90 0 0>}]
 clavicle_r [{shapeType=Capsule, length=6.25, radius=5.255, center=<-5 0 0>, rotation=<rot 90 0 0>}]
 dyn_snake_active_a_MSTR [{shapeType=Capsule, length=0.505, radius=0.505}]
 dyn_snake_active_b_06 [{shapeType=Capsule, length=2.2698, radius=1.9741, center=<4 0 0>, rotation=<rot 90 0 0>}]
 dyn_snake_active_b_09 [{shapeType=Capsule, length=3.7021, radius=2.1077, center=<1.5743 1.0705 0>, rotation=<rot 90 0 0>}]
 dyn_snake_active_b_MSTR [{shapeType=Capsule, length=0.505, radius=0.505}]
 dyn_snake_active_c_MSTR [{shapeType=Capsule, length=0.505, radius=0.505}]
 dyn_snake_active_d_03 [{shapeType=Capsule, length=3.4491, radius=1.8217, center=<1.5 0 0>, rotation=<rot 90 0 0>}]
 dyn_snake_active_d_06 [{shapeType=Capsule, length=3.4491, radius=1.8217, center=<1.5 0 0>, rotation=<rot 90 0 0>}]
 dyn_snake_active_d_08 [{shapeType=Capsule, length=3.4491, radius=1.8217, center=<1.5 0 0>, rotation=<rot 90 0 0>}]
 dyn_snake_active_d_MSTR [{shapeType=Capsule, length=0.505, radius=0.505}]
 dyn_snake_active_e_MSTR [{shapeType=Capsule, length=0.505, radius=0.505}]
 dyn_snake_active_f_03 [{shapeType=Capsule, length=4.2248, radius=2.225, center=<1.5 0 0>, rotation=<rot 90 0 0>}]
 dyn_snake_active_f_06 [{shapeType=Capsule, length=4.2248, radius=2.225, center=<1.5 -1.1756 -0.0001>, rotation=<rot 90 0 0>}]
 dyn_snake_active_f_MSTR [{shapeType=Capsule, length=0.505, radius=0.505}]
 spine_04 [{shapeType=Capsule, length=9.7013, radius=7.968, center=<0.6532 1.8714 0>}]
 spine_05 [{shapeType=Capsule, length=15.6163, radius=8.0469, center=<2.8332 0.6436 -0.0001>}]
}
constraints [
 {
  bone1Name=spine_05
  bone2Name=spine_04
  swing1LimitDegrees=45
  swing1Motion=Free
  swing2LimitDegrees=45
  swing2Motion=Free
  twistLimitDegrees=45
  twistMotion=Free
 }
 {
  bone1Name=clavicle_l
  bone2Name=spine_05
  swing1LimitDegrees=45
  swing1Motion=Free
  swing2LimitDegrees=45
  swing2Motion=Free
  twistLimitDegrees=45
  twistMotion=Free
 }
 {
  bone1Name=clavicle_r
  bone2Name=spine_05
  swing1LimitDegrees=45
  swing1Motion=Free
  swing2LimitDegrees=45
  swing2Motion=Free
  twistLimitDegrees=45
  twistMotion=Free
 }
 {
  bone1Name=dyn_snake_active_a_MSTR
  bone2Name=spine_05
  swing1LimitDegrees=45
  swing1Motion=Free
  swing2LimitDegrees=45
  swing2Motion=Free
  twistLimitDegrees=45
  twistMotion=Free
 }
 {
  bone1Name=dyn_snake_active_b_MSTR
  bone2Name=spine_05
  swing1LimitDegrees=45
  swing1Motion=Free
  swing2LimitDegrees=45
  swing2Motion=Free
  twistLimitDegrees=45
  twistMotion=Free
 }
 {
  bone1Name=dyn_snake_active_c_MSTR
  bone2Name=spine_05
  swing1LimitDegrees=45
  swing1Motion=Free
  swing2LimitDegrees=45
  swing2Motion=Free
  twistLimitDegrees=45
  twistMotion=Free
 }
 {
  bone1Name=dyn_snake_active_d_MSTR
  bone2Name=spine_05
  swing1LimitDegrees=45
  swing1Motion=Free
  swing2LimitDegrees=45
  swing2Motion=Free
  twistLimitDegrees=45
  twistMotion=Free
 }
 {bone1Name=dyn_snake_active_d_03, bone2Name=dyn_snake_active_d_MSTR, swing1LimitDegrees=10, swing2LimitDegrees=10, twistLimitDegrees=3}
 {bone1Name=dyn_snake_active_d_06, bone2Name=dyn_snake_active_d_03, swing1LimitDegrees=10, swing2LimitDegrees=10, twistLimitDegrees=3}
 {bone1Name=dyn_snake_active_d_08, bone2Name=dyn_snake_active_d_06, swing1LimitDegrees=10, swing2LimitDegrees=10, twistLimitDegrees=3}
 {
  bone1Name=dyn_snake_active_e_MSTR
  bone2Name=spine_05
  swing1LimitDegrees=45
  swing1Motion=Free
  swing2LimitDegrees=45
  swing2Motion=Free
  twistLimitDegrees=45
  twistMotion=Free
 }
 {
  bone1Name=dyn_snake_active_f_MSTR
  bone2Name=spine_05
  swing1LimitDegrees=45
  swing1Motion=Free
  swing2LimitDegrees=45
  swing2Motion=Free
  twistLimitDegrees=45
  twistMotion=Free
 }
 {bone1Name=dyn_snake_active_f_03, bone2Name=dyn_snake_active_f_MSTR, swing1LimitDegrees=35, swing2LimitDegrees=35, twistLimitDegrees=1}
 {bone1Name=dyn_snake_active_f_06, bone2Name=dyn_snake_active_f_03, swing1LimitDegrees=35, swing2LimitDegrees=20, twistLimitDegrees=1}
 {bone1Name=dyn_snake_active_b_06, bone2Name=dyn_snake_active_b_MSTR, swing1LimitDegrees=35, swing2LimitDegrees=35, twistLimitDegrees=5}
 {bone1Name=dyn_snake_active_b_09, bone2Name=dyn_snake_active_b_06, swing1LimitDegrees=35, swing2LimitDegrees=35, twistLimitDegrees=5}
]
