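# Physics asset for the Stone Viper female body's "Cheap" setup, for simulating six head-attached snake chains (a–f).
# Capsules on pelvis, spine, clavicles, upper arms and head, plus a broad, thin neck box, form the body collision shapes, with angularly free constraints.
# Each snake runs from a tiny MSTR capsule on the head through selected bones, with tight twist and per-chain swing limits.
# Chain roots swing freely, as does chain A's second link.
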PhysicsAsset /BRCosmetics/Characters/Player/Female/Medium/Bodies/F_MED_Stone_Viper/Meshes/Parts/F_MED_Stone_Viper_Cheap_Physics
bodies {
 clavicle_l [{shapeType=Capsule, length=3.6894, radius=7.0419, center=<6.4482 -0.3686 -2.4355>, rotation=<rot 85.606 0 0>}]
 clavicle_r [{shapeType=Capsule, length=3.4794, radius=7.2519, center=<-6.4482 0.3686 2.4355>, rotation=<rot 85.606 0 0>}]
 dyn_snake_a_02 [{shapeType=Capsule, length=0.505, radius=0.505, rotation=<rot 0 -90 0>}]
 dyn_snake_a_03 [{shapeType=Capsule, length=5, radius=1.5, center=<3 0 0>, rotation=<rot 90 0 0>}]
 dyn_snake_a_05 [{shapeType=Capsule, length=4.4343, radius=1.7533, center=<3 0 0>, rotation=<rot 90 0 0>}]
 dyn_snake_a_07 [{shapeType=Capsule, length=2.8027, radius=1.6111, center=<1 0 0>, rotation=<rot 90 0 0>}]
 dyn_snake_a_08 [{shapeType=Capsule, length=4.8291, radius=1.8119, center=<3.5 0 0>, rotation=<rot 90 0 0>}]
 dyn_snake_a_MSTR [{shapeType=Capsule, length=0.505, radius=0.505}]
 dyn_snake_b_02 [{shapeType=Capsule, length=2.5, radius=2, center=<2 0 0>, rotation=<rot 90 0 0>}]
 dyn_snake_b_05 [{shapeType=Capsule, length=4.7453, radius=1.5, center=<2 0 0>, rotation=<rot 90 0 0>}]
 dyn_snake_b_06 [{shapeType=Capsule, length=5, radius=1.5, center=<2.5 0 0>, rotation=<rot 90 0 0>}]
 dyn_snake_b_08 [{shapeType=Capsule, length=4.2534, radius=1.7309, center=<2.5 0 0>, rotation=<rot 90 0 0>}]
 dyn_snake_b_MSTR [{shapeType=Capsule, length=0.505, radius=0.505}]
 dyn_snake_c_02 [{shapeType=Capsule, length=2.5, radius=2, center=<2 0 0>, rotation=<rot 90 0 0>}]
 dyn_snake_c_05 [{shapeType=Capsule, length=4.3683, radius=1.5, center=<1.5 0 0>, rotation=<rot 90 0 0>}]
 dyn_snake_c_06 [{shapeType=Capsule, length=5, radius=1.5, center=<2.5 0 0>, rotation=<rot 90 0 0>}]
 dyn_snake_c_08 [{shapeType=Capsule, length=6.5127, radius=1.7309, center=<2.5 0 0>, rotation=<rot 90 0 0>}]
 dyn_snake_c_MSTR [{shapeType=Capsule, length=0.505, radius=0.505}]
 dyn_snake_d_02 [{shapeType=Capsule, length=2.5, radius=2, center=<2 0 0>, rotation=<rot 90 0 0>}]
 dyn_snake_d_05 [{shapeType=Capsule, length=5.8008, radius=1.6572, center=<2 0 0>, rotation=<rot 90 0 0>}]
 dyn_snake_d_06 [{shapeType=Capsule, length=5, radius=1.5, center=<2.5 0 0>, rotation=<rot 90 0 0>}]
 dyn_snake_d_08 [{shapeType=Capsule, length=4.5335, radius=2.1565, center=<2 0 0>, rotation=<rot 90 0 0>}]
 dyn_snake_d_MSTR [{shapeType=Capsule, length=0.505, radius=0.505}]
 dyn_snake_e_02 [{shapeType=Capsule, length=2.5, radius=2, center=<2 0 0>, rotation=<rot 90 0 0>}]
 dyn_snake_e_05 [{shapeType=Capsule, length=5.5798, radius=1.5, center=<2 0 0>, rotation=<rot 90 0 0>}]
 dyn_snake_e_06 [{shapeType=Capsule, length=5, radius=1.5, center=<2.5 0 0>, rotation=<rot 90 0 0>}]
 dyn_snake_e_08 [{shapeType=Capsule, length=4.2534, radius=1.7309, center=<2.5 0 0>, rotation=<rot 90 0 0>}]
 dyn_snake_e_MSTR [{shapeType=Capsule, length=0.505, radius=0.505}]
 dyn_snake_f_02 [{shapeType=Capsule, length=2.5, radius=2, center=<2 0 0>, rotation=<rot 90 0 0>}]
 dyn_snake_f_05 [{shapeType=Capsule, length=4.0982, radius=2.2096, center=<2.0001 -1 1.75>, rotation=<rot 90 0 0>}]
 dyn_snake_f_06 [{shapeType=Capsule, length=4.2103, radius=1.5, center=<2.5 0 0>, rotation=<rot 90 0 0>}]
 dyn_snake_f_08 [{shapeType=Capsule, length=4.2534, radius=1.7309, center=<2.5 0 0>, rotation=<rot 90 0 0>}]
 dyn_snake_f_MSTR [{shapeType=Capsule, length=0.505, radius=0.505}]
 head [{shapeType=Capsule, length=6.8955, radius=5.5858, center=<0.968 -0.4298 -0.1056>, rotation=<rot 90 90 84.9638>}]
 neck_01 [{shapeType=Box, extentX=28.9592, extentY=2.6107, extentZ=74.2313, center=<0.0972 -0.3584 0.659>, rotation=<rot 0 -5.0001 0>}]
 pelvis [{shapeType=Capsule, length=12.1901, radius=9.5301, center=<-4.2652 -2.555 0>}]
 spine_01 [{shapeType=Capsule, length=8.5377, radius=8.2027}]
 spine_02 [{shapeType=Capsule, length=5.9705, radius=8.255, center=<0 1.813 -0.0001>}]
 spine_03 [{shapeType=Capsule, length=7.9414, radius=7.7882, center=<0 2.0457 -0.0001>}]
 spine_04 [{shapeType=Capsule, length=9.7013, radius=7.968, center=<0.6532 1.8714 0>}]
 spine_05 [{shapeType=Capsule, length=15.7818, radius=7.8814, center=<2.8332 0.6436 -0.0001>}]
 upperarm_l [{shapeType=Capsule, length=18.1814, radius=4.9058, center=<10.8056 -0.4936 0.3456>, rotation=<rot 90 0 0>}]
 upperarm_r [{shapeType=Capsule, length=18.1814, radius=4.9058, center=<-10.8056 0.4936 -0.3456>, rotation=<rot 90 0 0>}]
}
constraints [
 {
  bone1Name=spine_05
  bone2Name=spine_04
  swing1LimitDegrees=45
  swing1Motion=Free
  swing2LimitDegrees=45
  swing2Motion=Free
  twistLimitDegrees=45
  twistMotion=Free
 }
 {
  bone1Name=spine_04
  bone2Name=spine_03
  swing1LimitDegrees=45
  swing1Motion=Free
  swing2LimitDegrees=45
  swing2Motion=Free
  twistLimitDegrees=45
  twistMotion=Free
 }
 {
  bone1Name=spine_03
  bone2Name=spine_02
  swing1LimitDegrees=45
  swing1Motion=Free
  swing2LimitDegrees=45
  swing2Motion=Free
  twistLimitDegrees=45
  twistMotion=Free
 }
 {
  bone1Name=dyn_snake_a_MSTR
  bone2Name=head
  swing1LimitDegrees=45
  swing1Motion=Free
  swing2LimitDegrees=45
  swing2Motion=Free
  twistLimitDegrees=45
  twistMotion=Free
 }
 {
  bone1Name=dyn_snake_b_MSTR
  bone2Name=head
  swing1LimitDegrees=45
  swing1Motion=Free
  swing2LimitDegrees=45
  swing2Motion=Free
  twistLimitDegrees=45
  twistMotion=Free
 }
 {
  bone1Name=dyn_snake_c_MSTR
  bone2Name=head
  swing1LimitDegrees=45
  swing1Motion=Free
  swing2LimitDegrees=45
  swing2Motion=Free
  twistLimitDegrees=45
  twistMotion=Free
 }
 {
  bone1Name=dyn_snake_d_MSTR
  bone2Name=head
  swing1LimitDegrees=45
  swing1Motion=Free
  swing2LimitDegrees=45
  swing2Motion=Free
  twistLimitDegrees=45
  twistMotion=Free
 }
 {
  bone1Name=dyn_snake_e_MSTR
  bone2Name=head
  swing1LimitDegrees=45
  swing1Motion=Free
  swing2LimitDegrees=45
  swing2Motion=Free
  twistLimitDegrees=45
  twistMotion=Free
 }
 {
  bone1Name=dyn_snake_f_MSTR
  bone2Name=head
  swing1LimitDegrees=45
  swing1Motion=Free
  swing2LimitDegrees=45
  swing2Motion=Free
  twistLimitDegrees=45
  twistMotion=Free
 }
 {
  bone1Name=head
  bone2Name=neck_01
  swing1LimitDegrees=45
  swing1Motion=Free
  swing2LimitDegrees=45
  swing2Motion=Free
  twistLimitDegrees=45
  twistMotion=Free
 }
 {
  bone1Name=clavicle_l
  bone2Name=spine_05
  swing1LimitDegrees=45
  swing1Motion=Free
  swing2LimitDegrees=45
  swing2Motion=Free
  twistLimitDegrees=45
  twistMotion=Free
 }
 {
  bone1Name=clavicle_r
  bone2Name=spine_05
  swing1LimitDegrees=45
  swing1Motion=Free
  swing2LimitDegrees=45
  swing2Motion=Free
  twistLimitDegrees=45
  twistMotion=Free
 }
 {
  bone1Name=upperarm_r
  bone2Name=clavicle_r
  swing1LimitDegrees=45
  swing1Motion=Free
  swing2LimitDegrees=45
  swing2Motion=Free
  twistLimitDegrees=45
  twistMotion=Free
 }
 {
  bone1Name=upperarm_l
  bone2Name=clavicle_l
  swing1LimitDegrees=45
  swing1Motion=Free
  swing2LimitDegrees=45
  swing2Motion=Free
  twistLimitDegrees=45
  twistMotion=Free
 }
 {bone1Name=dyn_snake_d_05, bone2Name=dyn_snake_d_02, swing1LimitDegrees=15, swing2LimitDegrees=45, twistLimitDegrees=5}
 {bone1Name=dyn_snake_d_08, bone2Name=dyn_snake_d_06, swing1LimitDegrees=15, swing2LimitDegrees=45, twistLimitDegrees=5}
 {bone1Name=dyn_snake_f_05, bone2Name=dyn_snake_f_02, swing1LimitDegrees=7, swing2LimitDegrees=35, twistLimitDegrees=5}
 {bone1Name=dyn_snake_e_05, bone2Name=dyn_snake_e_02, swing1LimitDegrees=15, swing2LimitDegrees=45, twistLimitDegrees=5}
 {bone1Name=dyn_snake_c_05, bone2Name=dyn_snake_c_02, swing1LimitDegrees=15, swing2LimitDegrees=45, twistLimitDegrees=5}
 {bone1Name=dyn_snake_b_05, bone2Name=dyn_snake_b_02, swing1LimitDegrees=15, swing2LimitDegrees=45, twistLimitDegrees=5}
 {bone1Name=dyn_snake_f_08, bone2Name=dyn_snake_f_06, swing1LimitDegrees=7, swing2LimitDegrees=35, twistLimitDegrees=5}
 {bone1Name=dyn_snake_e_08, bone2Name=dyn_snake_e_06, swing1LimitDegrees=15, swing2LimitDegrees=45, twistLimitDegrees=5}
 {bone1Name=dyn_snake_c_08, bone2Name=dyn_snake_c_06, swing1LimitDegrees=15, swing2LimitDegrees=45, twistLimitDegrees=5}
 {bone1Name=dyn_snake_b_08, bone2Name=dyn_snake_b_06, swing1LimitDegrees=15, swing2LimitDegrees=45, twistLimitDegrees=5}
 {bone1Name=dyn_snake_a_05, bone2Name=dyn_snake_a_03, swing1LimitDegrees=25, swing2LimitDegrees=35, twistLimitDegrees=5}
 {bone1Name=dyn_snake_a_08, bone2Name=dyn_snake_a_07, swing1LimitDegrees=35, swing2LimitDegrees=35, twistLimitDegrees=5}
 {bone1Name=dyn_snake_b_02, bone2Name=dyn_snake_b_MSTR, swing1LimitDegrees=10, swing2LimitDegrees=50, twistLimitDegrees=5}
 {bone1Name=dyn_snake_c_02, bone2Name=dyn_snake_c_MSTR, swing1LimitDegrees=10, swing2LimitDegrees=50, twistLimitDegrees=5}
 {bone1Name=dyn_snake_d_02, bone2Name=dyn_snake_d_MSTR, swing1LimitDegrees=10, swing2LimitDegrees=50, twistLimitDegrees=5}
 {bone1Name=dyn_snake_e_02, bone2Name=dyn_snake_e_MSTR, swing1LimitDegrees=10, swing2LimitDegrees=50, twistLimitDegrees=5}
 {bone1Name=dyn_snake_f_02, bone2Name=dyn_snake_f_MSTR, swing1LimitDegrees=7, swing2LimitDegrees=35, twistLimitDegrees=5}
 {bone1Name=dyn_snake_b_06, bone2Name=dyn_snake_b_05, swing1LimitDegrees=15, swing2LimitDegrees=45, twistLimitDegrees=5}
 {bone1Name=dyn_snake_c_06, bone2Name=dyn_snake_c_05, swing1LimitDegrees=15, swing2LimitDegrees=45, twistLimitDegrees=5}
 {bone1Name=dyn_snake_d_06, bone2Name=dyn_snake_d_05, swing1LimitDegrees=15, swing2LimitDegrees=45, twistLimitDegrees=5}
 {bone1Name=dyn_snake_e_06, bone2Name=dyn_snake_e_05, swing1LimitDegrees=15, swing2LimitDegrees=45, twistLimitDegrees=5}
 {bone1Name=dyn_snake_f_06, bone2Name=dyn_snake_f_05, swing1LimitDegrees=7, swing2LimitDegrees=35, twistLimitDegrees=5}
 {
  bone1Name=neck_01
  bone2Name=spine_05
  swing1LimitDegrees=45
  swing1Motion=Free
  swing2LimitDegrees=45
  swing2Motion=Free
  twistLimitDegrees=45
  twistMotion=Free
 }
 {
  bone1Name=spine_02
  bone2Name=spine_01
  swing1LimitDegrees=45
  swing1Motion=Free
  swing2LimitDegrees=45
  swing2Motion=Free
  twistLimitDegrees=45
  twistMotion=Free
 }
 {
  bone1Name=spine_01
  bone2Name=pelvis
  swing1LimitDegrees=45
  swing1Motion=Free
  swing2LimitDegrees=45
  swing2Motion=Free
  twistLimitDegrees=45
  twistMotion=Free
 }
 {bone1Name=dyn_snake_a_07, bone2Name=dyn_snake_a_05, swing1LimitDegrees=35, swing2LimitDegrees=35, twistLimitDegrees=5}
 {bone1Name=dyn_snake_a_03, bone2Name=dyn_snake_a_02, swing1LimitDegrees=20, swing2LimitDegrees=30, twistLimitDegrees=5}
 {
  bone1Name=dyn_snake_a_02
  bone2Name=dyn_snake_a_MSTR
  swing1LimitDegrees=45
  swing1Motion=Free
  swing2LimitDegrees=45
  swing2Motion=Free
  twistLimitDegrees=45
  twistMotion=Free
 }
]
