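# Physics asset for the front end mesh part of the Stone Viper female medium character, covering six dynamic snake bone chains on the head.
# Each chain (a–f) runs capsules from a small MSTR body, free-swinging on the head, through seven or eight segments.
# Segment limits mostly allow bending in one swing axis with little twist; chains b and f are stiffer at the root.
# Spine, clavicle, upper arm, head and boxed neck bodies form a freely constrained upper-body proxy.
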PhysicsAsset /BRCosmetics/Characters/Player/Female/Medium/Bodies/F_MED_Stone_Viper/Meshes/Parts/F_MED_Stone_Viper_Front_End_Physics
bodies {
 clavicle_l [{shapeType=Capsule, length=3.4794, radius=7.2519, center=<6.4482 -0.3686 -2.4355>, rotation=<rot 85.606 0 0>}]
 clavicle_r [{shapeType=Capsule, length=3.4794, radius=7.2519, center=<-6.4482 0.3686 2.4355>, rotation=<rot 85.606 0 0>}]
 dyn_snake_a_01 [{shapeType=Capsule, length=2, radius=1.5, center=<2 0 0>, rotation=<rot 90 0 0>}]
 dyn_snake_a_02 [{shapeType=Capsule, length=2, radius=1.5, center=<2 0 0>, rotation=<rot 90 0 0>}]
 dyn_snake_a_03 [{shapeType=Capsule, length=2, radius=1.5, center=<2 0 0>, rotation=<rot 90 0 0>}]
 dyn_snake_a_04 [{shapeType=Capsule, length=2, radius=1.5, center=<2 0 0>, rotation=<rot 90 0 0>}]
 dyn_snake_a_05 [{shapeType=Capsule, length=3.6056, radius=1.5, center=<2 0 0>, rotation=<rot 90 0 0>}]
 dyn_snake_a_06 [{shapeType=Capsule, length=2, radius=1.5, center=<2 0 0>, rotation=<rot 90 0 0>}]
 dyn_snake_a_07 [{shapeType=Capsule, length=3.6855, radius=1.7973, center=<2 0 0>, rotation=<rot 90 0 0>}]
 dyn_snake_a_MSTR [{shapeType=Capsule, length=0.505, radius=0.505}]
 dyn_snake_b_01 [{shapeType=Capsule, length=2, radius=1.5, center=<2 0 0>, rotation=<rot 90 0 0>}]
 dyn_snake_b_02 [{shapeType=Capsule, length=2, radius=1.5, center=<2 0 0>, rotation=<rot 90 0 0>}]
 dyn_snake_b_03 [{shapeType=Capsule, length=2, radius=1.5, center=<2 0 0>, rotation=<rot 90 0 0>}]
 dyn_snake_b_04 [{shapeType=Capsule, length=2, radius=1.5, center=<2 0 0>, rotation=<rot 90 0 0>}]
 dyn_snake_b_05 [{shapeType=Capsule, length=3.6056, radius=1.5, center=<2 0 0>, rotation=<rot 90 0 0>}]
 dyn_snake_b_06 [{shapeType=Capsule, length=2, radius=1.5, center=<2 0 0>, rotation=<rot 90 0 0>}]
 dyn_snake_b_07 [{shapeType=Capsule, length=3.7899, radius=1.693, center=<2 0 0>, rotation=<rot 90 0 0>}]
 dyn_snake_b_MSTR [{shapeType=Capsule, length=0.505, radius=0.505}]
 dyn_snake_c_01 [{shapeType=Capsule, length=2, radius=1.5, center=<2 0 0>, rotation=<rot 90 0 0>}]
 dyn_snake_c_02 [{shapeType=Capsule, length=2, radius=1.5, center=<2 0 0>, rotation=<rot 90 0 0>}]
 dyn_snake_c_03 [{shapeType=Capsule, length=2, radius=1.5, center=<2 0 0>, rotation=<rot 90 0 0>}]
 dyn_snake_c_04 [{shapeType=Capsule, length=2, radius=1.5, center=<2 0 0>, rotation=<rot 90 0 0>}]
 dyn_snake_c_05 [{shapeType=Capsule, length=3.6056, radius=1.5, center=<2 0 0>, rotation=<rot 90 0 0>}]
 dyn_snake_c_06 [{shapeType=Capsule, length=2, radius=1.5, center=<2 0 0>, rotation=<rot 90 0 0>}]
 dyn_snake_c_07 [{shapeType=Capsule, length=3.7899, radius=1.693, center=<2 0 0>, rotation=<rot 90 0 0>}]
 dyn_snake_c_MSTR [{shapeType=Capsule, length=0.505, radius=0.505}]
 dyn_snake_d_01 [{shapeType=Capsule, length=2, radius=1.5, center=<2 0 0>, rotation=<rot 90 0 0>}]
 dyn_snake_d_02 [{shapeType=Capsule, length=2, radius=1.5, center=<2 0 0>, rotation=<rot 90 0 0>}]
 dyn_snake_d_03 [{shapeType=Capsule, length=2, radius=1.5, center=<2 0 0>, rotation=<rot 90 0 0>}]
 dyn_snake_d_04 [{shapeType=Capsule, length=2, radius=1.5, center=<2 0 0>, rotation=<rot 90 0 0>}]
 dyn_snake_d_05 [{shapeType=Capsule, length=3.6056, radius=1.5, center=<2 0 0>, rotation=<rot 90 0 0>}]
 dyn_snake_d_06 [{shapeType=Capsule, length=2, radius=1.5, center=<2 0 0>, rotation=<rot 90 0 0>}]
 dyn_snake_d_07 [{shapeType=Capsule, length=3.7899, radius=1.693, center=<2 0 0>, rotation=<rot 90 0 0>}]
 dyn_snake_d_MSTR [{shapeType=Capsule, length=0.505, radius=0.505}]
 dyn_snake_e_01 [{shapeType=Capsule, length=2, radius=1.5, center=<2 0 0>, rotation=<rot 90 0 0>}]
 dyn_snake_e_02 [{shapeType=Capsule, length=2, radius=1.5, center=<2 0 0>, rotation=<rot 90 0 0>}]
 dyn_snake_e_03 [{shapeType=Capsule, length=2, radius=1.5, center=<2 0 0>, rotation=<rot 90 0 0>}]
 dyn_snake_e_04 [{shapeType=Capsule, length=2, radius=1.5, center=<2 0 0>, rotation=<rot 90 0 0>}]
 dyn_snake_e_05 [{shapeType=Capsule, length=2, radius=1.5, center=<2 0 0>, rotation=<rot 90 0 0>}]
 dyn_snake_e_06 [{shapeType=Capsule, length=2, radius=1.5, center=<2 0 0>, rotation=<rot 90 0 0>}]
 dyn_snake_e_07 [{shapeType=Capsule, length=3.7899, radius=1.693, center=<3.5 0 0>, rotation=<rot 90 0 0>}]
 dyn_snake_e_MSTR [{shapeType=Capsule, length=0.505, radius=0.505}]
 dyn_snake_f_01 [{shapeType=Capsule, length=2.099, radius=1.401, center=<2 0 0>, rotation=<rot 90 0 0>}]
 dyn_snake_f_02 [{shapeType=Capsule, length=2.099, radius=1.401, center=<2 0 0.5>, rotation=<rot 90 0 0>}]
 dyn_snake_f_03 [{shapeType=Capsule, length=2.099, radius=1.401, center=<2 0 0.75>, rotation=<rot 90 0 0>}]
 dyn_snake_f_04 [{shapeType=Capsule, length=2.1095, radius=1.3905, center=<2 0 1>, rotation=<rot 90 0 0>}]
 dyn_snake_f_05 [{shapeType=Capsule, length=2.1071, radius=1.3929, center=<2 0 0.75>, rotation=<rot 90 0 0>}]
 dyn_snake_f_06 [{shapeType=Capsule, length=2.1765, radius=1.3235, center=<2 0 0.5>, rotation=<rot 90 0 0>}]
 dyn_snake_f_07 [{shapeType=Capsule, length=3.2077, radius=2.2751, center=<3.5 0 0>, rotation=<rot 90 0 0>}]
 dyn_snake_f_08 [{shapeType=Capsule, length=2.4804, radius=2.64, center=<3.5 0 0>, rotation=<rot 90 0 0>}]
 dyn_snake_f_MSTR [{shapeType=Capsule, length=0.505, radius=0.505}]
 head [{shapeType=Capsule, length=6.8955, radius=5.5858, center=<0.968 -0.4298 -0.1056>, rotation=<rot 90 90 84.9638>}]
 neck_01 [{shapeType=Box, extentX=29.01, extentY=3.01, extentZ=33.01, center=<9.5794 0 0>}]
 spine_01 [{shapeType=Capsule, length=7.0613, radius=9.6791, center=<0 1.0381 -0.0001>}]
 spine_02 [{shapeType=Capsule, length=5.9705, radius=8.255, center=<0 1.813 -0.0001>}]
 spine_03 [{shapeType=Capsule, length=7.9414, radius=7.7882, center=<0 2.0457 -0.0001>}]
 spine_04 [{shapeType=Capsule, length=9.7013, radius=7.968, center=<0.6532 1.8714 0>}]
 spine_05 [{shapeType=Capsule, length=15.7818, radius=7.8814, center=<2.8332 0.6436 -0.0001>}]
 upperarm_l [{shapeType=Capsule, length=14.3295, radius=8.5935, center=<10.8056 -0.4936 0.3456>, rotation=<rot 90 0 0>}]
 upperarm_r [{shapeType=Capsule, length=18.0812, radius=4.1347, center=<-10.8056 0.4936 -0.3456>, rotation=<rot 90 0 0>}]
}
constraints [
 {
  bone1Name=spine_05
  bone2Name=spine_04
  swing1LimitDegrees=45
  swing1Motion=Free
  swing2LimitDegrees=45
  swing2Motion=Free
  twistLimitDegrees=45
  twistMotion=Free
 }
 {
  bone1Name=spine_04
  bone2Name=spine_03
  swing1LimitDegrees=45
  swing1Motion=Free
  swing2LimitDegrees=45
  swing2Motion=Free
  twistLimitDegrees=45
  twistMotion=Free
 }
 {
  bone1Name=spine_03
  bone2Name=spine_02
  swing1LimitDegrees=45
  swing1Motion=Free
  swing2LimitDegrees=45
  swing2Motion=Free
  twistLimitDegrees=45
  twistMotion=Free
 }
 {
  bone1Name=spine_02
  bone2Name=spine_01
  swing1LimitDegrees=45
  swing1Motion=Free
  swing2LimitDegrees=45
  swing2Motion=Free
  twistLimitDegrees=45
  twistMotion=Free
 }
 {
  bone1Name=dyn_snake_a_MSTR
  bone2Name=head
  swing1LimitDegrees=45
  swing1Motion=Free
  swing2LimitDegrees=45
  swing2Motion=Free
  twistLimitDegrees=45
  twistMotion=Free
 }
 {bone1Name=dyn_snake_a_01, bone2Name=dyn_snake_a_MSTR, swing1LimitDegrees=15, swing2LimitDegrees=10, twistLimitDegrees=5}
 {bone1Name=dyn_snake_a_02, bone2Name=dyn_snake_a_01, swing1LimitDegrees=15, swing2LimitDegrees=10, twistLimitDegrees=5}
 {bone1Name=dyn_snake_a_03, bone2Name=dyn_snake_a_02, swing1LimitDegrees=3, swing2LimitDegrees=35, twistLimitDegrees=3}
 {bone1Name=dyn_snake_a_04, bone2Name=dyn_snake_a_03, swing1LimitDegrees=3, swing2LimitDegrees=35, twistLimitDegrees=3}
 {bone1Name=dyn_snake_a_05, bone2Name=dyn_snake_a_04, swing1LimitDegrees=3, swing2LimitDegrees=35, twistLimitDegrees=3}
 {bone1Name=dyn_snake_a_06, bone2Name=dyn_snake_a_05, swing1LimitDegrees=3, swing2LimitDegrees=35, twistLimitDegrees=3}
 {bone1Name=dyn_snake_a_07, bone2Name=dyn_snake_a_06, swing1LimitDegrees=3, swing2LimitDegrees=35, twistLimitDegrees=3}
 {
  bone1Name=dyn_snake_b_MSTR
  bone2Name=head
  swing1LimitDegrees=45
  swing1Motion=Free
  swing2LimitDegrees=45
  swing2Motion=Free
  twistLimitDegrees=45
  twistMotion=Free
 }
 {bone1Name=dyn_snake_b_01, bone2Name=dyn_snake_b_MSTR, swing1LimitDegrees=3, swing2LimitDegrees=5, twistLimitDegrees=3}
 {bone1Name=dyn_snake_b_02, bone2Name=dyn_snake_b_01, swing1LimitDegrees=3, swing2LimitDegrees=10, twistLimitDegrees=3}
 {bone1Name=dyn_snake_b_03, bone2Name=dyn_snake_b_02, swing1LimitDegrees=3, swing2LimitDegrees=15, twistLimitDegrees=3}
 {bone1Name=dyn_snake_b_04, bone2Name=dyn_snake_b_03, swing1LimitDegrees=3, swing2LimitDegrees=25, twistLimitDegrees=3}
 {bone1Name=dyn_snake_b_05, bone2Name=dyn_snake_b_04, swing1LimitDegrees=3, swing2LimitDegrees=35, twistLimitDegrees=3}
 {bone1Name=dyn_snake_b_06, bone2Name=dyn_snake_b_05, swing1LimitDegrees=3, swing2LimitDegrees=35, twistLimitDegrees=3}
 {bone1Name=dyn_snake_b_07, bone2Name=dyn_snake_b_06, swing1LimitDegrees=3, swing2LimitDegrees=35, twistLimitDegrees=3}
 {
  bone1Name=dyn_snake_c_MSTR
  bone2Name=head
  swing1LimitDegrees=45
  swing1Motion=Free
  swing2LimitDegrees=45
  swing2Motion=Free
  twistLimitDegrees=45
  twistMotion=Free
 }
 {bone1Name=dyn_snake_c_01, bone2Name=dyn_snake_c_MSTR, swing1LimitDegrees=3, swing2LimitDegrees=35, twistLimitDegrees=3}
 {bone1Name=dyn_snake_c_02, bone2Name=dyn_snake_c_01, swing1LimitDegrees=3, swing2LimitDegrees=35, twistLimitDegrees=3}
 {bone1Name=dyn_snake_c_03, bone2Name=dyn_snake_c_02, swing1LimitDegrees=3, swing2LimitDegrees=35, twistLimitDegrees=3}
 {bone1Name=dyn_snake_c_04, bone2Name=dyn_snake_c_03, swing1LimitDegrees=3, swing2LimitDegrees=35, twistLimitDegrees=3}
 {bone1Name=dyn_snake_c_05, bone2Name=dyn_snake_c_04, swing1LimitDegrees=3, swing2LimitDegrees=35, twistLimitDegrees=3}
 {bone1Name=dyn_snake_c_06, bone2Name=dyn_snake_c_05, swing1LimitDegrees=3, swing2LimitDegrees=35, twistLimitDegrees=3}
 {bone1Name=dyn_snake_c_07, bone2Name=dyn_snake_c_06, swing1LimitDegrees=3, swing2LimitDegrees=35, twistLimitDegrees=3}
 {
  bone1Name=dyn_snake_d_MSTR
  bone2Name=head
  swing1LimitDegrees=45
  swing1Motion=Free
  swing2LimitDegrees=45
  swing2Motion=Free
  twistLimitDegrees=45
  twistMotion=Free
 }
 {bone1Name=dyn_snake_d_01, bone2Name=dyn_snake_d_MSTR, swing1LimitDegrees=3, swing2LimitDegrees=35, twistLimitDegrees=3}
 {bone1Name=dyn_snake_d_02, bone2Name=dyn_snake_d_01, swing1LimitDegrees=3, swing2LimitDegrees=35, twistLimitDegrees=3}
 {bone1Name=dyn_snake_d_03, bone2Name=dyn_snake_d_02, swing1LimitDegrees=3, swing2LimitDegrees=35, twistLimitDegrees=3}
 {bone1Name=dyn_snake_d_04, bone2Name=dyn_snake_d_03, swing1LimitDegrees=3, swing2LimitDegrees=35, twistLimitDegrees=3}
 {bone1Name=dyn_snake_d_05, bone2Name=dyn_snake_d_04, swing1LimitDegrees=3, swing2LimitDegrees=35, twistLimitDegrees=3}
 {bone1Name=dyn_snake_d_06, bone2Name=dyn_snake_d_05, swing1LimitDegrees=3, swing2LimitDegrees=35, twistLimitDegrees=3}
 {bone1Name=dyn_snake_d_07, bone2Name=dyn_snake_d_06, swing1LimitDegrees=3, swing2LimitDegrees=35, twistLimitDegrees=3}
 {
  bone1Name=dyn_snake_e_MSTR
  bone2Name=head
  swing1LimitDegrees=45
  swing1Motion=Free
  swing2LimitDegrees=45
  swing2Motion=Free
  twistLimitDegrees=45
  twistMotion=Free
 }
 {bone1Name=dyn_snake_e_01, bone2Name=dyn_snake_e_MSTR, swing1LimitDegrees=3, swing2LimitDegrees=35, twistLimitDegrees=3}
 {bone1Name=dyn_snake_e_02, bone2Name=dyn_snake_e_01, swing1LimitDegrees=3, swing2LimitDegrees=35, twistLimitDegrees=3}
 {bone1Name=dyn_snake_e_03, bone2Name=dyn_snake_e_02, swing1LimitDegrees=3, swing2LimitDegrees=35, twistLimitDegrees=3}
 {bone1Name=dyn_snake_e_04, bone2Name=dyn_snake_e_03, swing1LimitDegrees=3, swing2LimitDegrees=35, twistLimitDegrees=3}
 {bone1Name=dyn_snake_e_05, bone2Name=dyn_snake_e_04, swing1LimitDegrees=3, swing2LimitDegrees=35, twistLimitDegrees=3}
 {bone1Name=dyn_snake_e_06, bone2Name=dyn_snake_e_05, swing1LimitDegrees=3, swing2LimitDegrees=35, twistLimitDegrees=3}
 {bone1Name=dyn_snake_e_07, bone2Name=dyn_snake_e_06, swing1LimitDegrees=3, swing2LimitDegrees=35, twistLimitDegrees=3}
 {
  bone1Name=dyn_snake_f_MSTR
  bone2Name=head
  swing1LimitDegrees=45
  swing1Motion=Free
  swing2LimitDegrees=45
  swing2Motion=Free
  twistLimitDegrees=45
  twistMotion=Free
 }
 {bone1Name=dyn_snake_f_01, bone2Name=dyn_snake_f_MSTR, swing1LimitDegrees=3, swing2LimitDegrees=5, twistLimitDegrees=3}
 {bone1Name=dyn_snake_f_02, bone2Name=dyn_snake_f_01, swing1LimitDegrees=3, swing2LimitDegrees=10, twistLimitDegrees=3}
 {bone1Name=dyn_snake_f_03, bone2Name=dyn_snake_f_02, swing1LimitDegrees=3, swing2LimitDegrees=15, twistLimitDegrees=3}
 {bone1Name=dyn_snake_f_04, bone2Name=dyn_snake_f_03, swing1LimitDegrees=3, swing2LimitDegrees=25, twistLimitDegrees=3}
 {bone1Name=dyn_snake_f_05, bone2Name=dyn_snake_f_04, swing1LimitDegrees=3, swing2LimitDegrees=35, twistLimitDegrees=3}
 {bone1Name=dyn_snake_f_06, bone2Name=dyn_snake_f_05, swing1LimitDegrees=3, swing2LimitDegrees=35, twistLimitDegrees=3}
 {bone1Name=dyn_snake_f_07, bone2Name=dyn_snake_f_06, swing1LimitDegrees=3, swing2LimitDegrees=35, twistLimitDegrees=3}
 {
  bone1Name=head
  bone2Name=neck_01
  swing1LimitDegrees=45
  swing1Motion=Free
  swing2LimitDegrees=45
  swing2Motion=Free
  twistLimitDegrees=45
  twistMotion=Free
 }
 {
  bone1Name=neck_01
  bone2Name=spine_05
  swing1LimitDegrees=45
  swing1Motion=Free
  swing2LimitDegrees=45
  swing2Motion=Free
  twistLimitDegrees=45
  twistMotion=Free
 }
 {
  bone1Name=clavicle_l
  bone2Name=spine_05
  swing1LimitDegrees=45
  swing1Motion=Free
  swing2LimitDegrees=45
  swing2Motion=Free
  twistLimitDegrees=45
  twistMotion=Free
 }
 {
  bone1Name=clavicle_r
  bone2Name=spine_05
  swing1LimitDegrees=45
  swing1Motion=Free
  swing2LimitDegrees=45
  swing2Motion=Free
  twistLimitDegrees=45
  twistMotion=Free
 }
 {
  bone1Name=upperarm_r
  bone2Name=clavicle_r
  swing1LimitDegrees=45
  swing1Motion=Free
  swing2LimitDegrees=45
  swing2Motion=Free
  twistLimitDegrees=45
  twistMotion=Free
 }
 {
  bone1Name=upperarm_l
  bone2Name=clavicle_l
  swing1LimitDegrees=45
  swing1Motion=Free
  swing2LimitDegrees=45
  swing2Motion=Free
  twistLimitDegrees=45
  twistMotion=Free
 }
 {bone1Name=dyn_snake_f_08, bone2Name=dyn_snake_f_07, swing1LimitDegrees=3, swing2LimitDegrees=35, twistLimitDegrees=3}
]
