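# Character-part AnimBlueprint for the Stone Viper cosmetic that animates and simulates its hair, switching from passive hair to active snakes while the player is targeting.
# It layers snake animations onto the copied body pose, then applies Control Rigs, with extras for melee and crouch.
# Additive bone offsets and three rigid-body simulations follow, with further skydive offsets.
# The update event interpolates bone scales to shrink one hair set and grow the other, and sets simulation alphas and gravity.
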AnimBlueprint /BRCosmetics/Characters/Player/Female/Medium/Bodies/F_MED_Stone_Viper/Meshes/Parts/F_MED_Stone_Viper_AnimBP
target_skeleton=&/F_MED_Stone_Viper_Skeleton
graphs [
 AnimGraph {nodes [
   Apply Additive() {alphaBoolBlend {blendOption=Linear}}
   Local To Component(LocalPose=Use cached pose 'Stone_Viper'())
   RigidBody() {alphaBoolBlend {blendOption=Linear}}
   RigidBody() {
    bClampLinearTranslationLimitToRefPose=true
    bOverrideWorldGravity=true
    overridePhysicsAsset=/Game/Characters/Player/Female/Medium/Bodies/F_MED_Stone_Viper/Meshes/Parts/F_MED_Stone_Viper_Physics
    alphaBoolBlend {blendOption=Linear}
    baseBoneRef {boneName=neck_01}
    componentAppliedLinearAccClamp=<1000 1000 1000>
    componentLinearAccScale=<0.3 0.3 0.5>
    componentLinearVelScale=<0.25 0.15 0.25>
    overrideWorldGravity {z=-980}
   }
   AnimGraphNode_StateMachine_0 = Locomotion() {is=title AnimGraphNode_StateMachine}
   n1 = Layered blend per bone!(
    BasePose=Copy Pose From Mesh(SourceMeshComponent=Get Body_Mesh())
    BlendPoses_0=@AnimGraphNode_StateMachine_0
    BlendPoses_1:Pose Link Structure
    BlendWeights_0
    BlendWeights_1:float=1
   ) -> (Pose) {
    blendPoses [{linkId=-1, sourceLinkId=-1}, {linkId=-1, sourceLinkId=-1}]
    blendWeights [1,1]
    layerSetup [{branchFilters [{boneName=dyn_passive_MSTR}, {boneName=dyn_active_MSTR}]}, {branchFilters []}]
   }
   n2 = Copy Pose From Mesh(SourceMeshComponent=Get Body_Mesh()) {bUseAttachedParent=true}
   n3 = Sequence Player 'Stone_Viper_Loop'() {sequence=&/Stone_Viper_Loop}
   n4 = Control Rig(Source=Blend Poses by bool(
     BlendPose_0=@n1
     BlendPose_1=Layered blend per bone!(BasePose=@n2, BlendPoses_0=@n3, BlendWeights_0) -> (Pose) {layerSetup [{branchFilters [{boneName=dyn_passive_MSTR}, {boneName=dyn_active_MSTR}]}]}
     bActiveValue=Get bIsTargeting()
    ) {blendType=Linear}) {
    alphaBoolBlend {blendOption=Linear}
    controlRigAssetReference {blueprintRigClass=&/F_MED_Stone_Viper_CtrlRig.F_MED_Stone_Viper_CtrlRig_C}
   }
   n5 = Control Rig!(Source=@n4, bAlphaBoolEnabled:bool=Get bIsPlayingMeleeAnim()) -> (Pose) {
    alphaInputType=Bool
    alphaBoolBlend {blendInTime=0.1, blendOption=Linear, blendOutTime=0.4}
    controlRigAssetReference {blueprintRigClass=&/F_MED_Stone_Viper_Axe_CtrlRig.F_MED_Stone_Viper_Axe_CtrlRig_C}
   }
   n6 = Control Rig(Source=@n5, bAlphaBoolEnabled=Get bIsCrouching()) {
    alphaInputType=Bool
    alphaBoolBlend {blendInTime=0.1, blendOption=Linear, blendOutTime=0.1}
    controlRigAssetReference {blueprintRigClass=&/F_MED_Stone_Viper_Crouch_CtrlRig.F_MED_Stone_Viper_Crouch_CtrlRig_C}
   }
   n7 = Get bIsPlayingEmote()
   n8 = "Transform (Modify) Bone - Bone: dyn_snake_a_02"!(
    ComponentPose=Local To Component(LocalPose=@n6)
    bAlphaBoolEnabled:bool=@n7
    Translation="0,0,0"
    Rotation="-3.247,15.841,24.235"
    Scale="1,1,1"
   ) -> (Pose) {
    alphaInputType=Bool
    rotationMode=BMM_Additive
    alphaBoolBlend {blendInTime=0.5, blendOption=Linear, blendOutTime=0.2}
    boneToModify {boneName=dyn_snake_a_02}
    rotation=<rot -3.247 15.841 24.235>
   }
   n9 = "Transform (Modify) Bone - Bone: dyn_snake_a_03"!(ComponentPose=@n8, bAlphaBoolEnabled:bool=@n7, Translation="0,0,0", Rotation="0.000,41.581,24.566", Scale="1,1,1") -> (Pose) {
    alphaInputType=Bool
    rotationMode=BMM_Additive
    alphaBoolBlend {blendInTime=0.5, blendOption=Linear, blendOutTime=0.5}
    boneToModify {boneName=dyn_snake_a_03}
    rotation {roll=24.5657, yaw=41.5812}
   }
   n10 = "Transform (Modify) Bone - Bone: dyn_snake_b_MSTR"!(ComponentPose=@n9, bAlphaBoolEnabled:bool=@n7, Translation="0,0,0", Rotation="0.000,0.000,19.025", Scale="1,1,1") -> (Pose) {
    alphaInputType=Bool
    rotationMode=BMM_Additive
    alphaBoolBlend {blendInTime=0.5, blendOption=Linear, blendOutTime=0.5}
    boneToModify {boneName=dyn_snake_b_MSTR}
    rotation {roll=19.0252}
   }
   n11 = "Transform (Modify) Bone - Bone: dyn_snake_c_MSTR"!(ComponentPose=@n10, bAlphaBoolEnabled:bool=@n7, Translation="0,0,0", Rotation="0.000,0.000,19.025", Scale="1,1,1") -> (Pose) {
    alphaInputType=Bool
    rotationMode=BMM_Additive
    alphaBoolBlend {blendInTime=0.5, blendOption=Linear, blendOutTime=0.5}
    boneToModify {boneName=dyn_snake_c_MSTR}
    rotation {roll=19.0252}
   }
   n12 = "Transform (Modify) Bone - Bone: dyn_snake_d_MSTR"!(ComponentPose=@n11, bAlphaBoolEnabled:bool=@n7, Translation="0,0,0", Rotation="0.000,0.000,19.025", Scale="1,1,1") -> (Pose) {
    alphaInputType=Bool
    rotationMode=BMM_Additive
    alphaBoolBlend {blendInTime=0.5, blendOption=Linear, blendOutTime=0.5}
    boneToModify {boneName=dyn_snake_d_MSTR}
    rotation {roll=19.0252}
   }
   n13 = "Transform (Modify) Bone - Bone: dyn_snake_e_MSTR"!(ComponentPose=@n12, bAlphaBoolEnabled:bool=@n7, Translation="0,0,0", Rotation="0.000,0.000,19.025", Scale="1,1,1") -> (Pose) {
    alphaInputType=Bool
    rotationMode=BMM_Additive
    alphaBoolBlend {blendInTime=0.5, blendOption=Linear, blendOutTime=0.5}
    boneToModify {boneName=dyn_snake_e_MSTR}
    rotation {roll=19.0252}
   }
   n14 = "Transform (Modify) Bone - Bone: dyn_snake_f_MSTR"!(ComponentPose=@n13, bAlphaBoolEnabled:bool=True, Translation="0,0,0", Rotation="0.000,0.000,15", Scale="1,1,1") -> (Pose) {
    alphaInputType=Bool
    rotationMode=BMM_Additive
    alphaBoolBlend {blendInTime=0.5, blendOption=Linear, blendOutTime=0.5}
    boneToModify {boneName=dyn_snake_f_MSTR}
    rotation {roll=15}
   }
   n15 = "Transform (Modify) Bone - Bone: dyn_snake_active_f_MSTR"(ComponentPose=@n14, Alpha=1, Translation="0,0,0", Rotation="0.386,8.400,-9.795", Scale="1,1,1") {
    rotationMode=BMM_Additive
    alphaBoolBlend {blendOption=Linear}
    boneToModify {boneName=dyn_snake_active_f_MSTR}
    rotation=<rot 0.386 8.4001 -9.7951>
   }
   n16 = "Transform (Modify) Bone - Bone: dyn_snake_active_e_MSTR"(ComponentPose=@n15, Alpha=1, Translation="0,0,0", Rotation="-11.837,9.734,8.947", Scale="1,1,1") {
    rotationMode=BMM_Additive
    alphaBoolBlend {blendOption=Linear}
    boneToModify {boneName=dyn_snake_active_e_MSTR}
    rotation=<rot -11.837 9.7341 8.9471>
   }
   n17 = "Transform (Modify) Bone - Bone: dyn_snake_active_d_MSTR"(ComponentPose=@n16, Alpha=1, Translation="0,0,0", Rotation="3.114,4.640,10.173", Scale="1,1,1") {
    rotationMode=BMM_Additive
    alphaBoolBlend {blendOption=Linear}
    boneToModify {boneName=dyn_snake_active_d_MSTR}
    rotation=<rot 3.114 4.6401 10.1734>
   }
   n18 = RigidBody!(
    ComponentPose=@n17
    Alpha=Get Hair_RBAN_AlphaInterp() -> (Hair_RBAN_AlphaInterp:double)
    OverrideWorldGravity:Vector=Get GravityOverride()
    ExternalForce
    ComponentLinearAccScale:Vector=Get RigidsAcc() -> (RigidsAcc:Vector)
    ComponentLinearVelScale:Vector=Get RigidsVel() -> (RigidsVel:Vector)
   ) -> (Pose) {
    bClampLinearTranslationLimitToRefPose=true
    bOverrideWorldGravity=true
    overridePhysicsAsset=&/F_MED_Stone_Viper_Cheap_Physics
    simulationSpace=BaseBoneSpace
    alphaBoolBlend {blendInTime=0.5, blendOption=Linear, blendOutTime=0.5}
    baseBoneRef {boneName=spine_05}
    componentAppliedLinearAccClamp=<5000 5000 5000>
    componentLinearAccScale=<0.25 0.6 0.3>
    componentLinearVelScale=<0.15 0.45 0.15>
    overrideWorldGravity {z=-980}
   }
   n19 = RigidBody!(ComponentPose=@n18, Alpha=Get Rigids_Alpha() -> (Rigids_Alpha:double), ExternalForce) -> (Pose) {
    bClampLinearTranslationLimitToRefPose=true
    bOverrideWorldGravity=true
    overridePhysicsAsset=&/F_MED_Stone_Viper_Cheap_Active_Physics
    simulationSpace=BaseBoneSpace
    alphaBoolBlend {blendInTime=0.5, blendOption=Linear, blendOutTime=0.5}
    baseBoneRef {boneName=spine_05}
    componentAppliedLinearAccClamp=<2000 2000 1000>
    componentLinearAccScale=<0.5 0.3 0.4>
    componentLinearVelScale=<0.3 0.25 0.2>
    overrideWorldGravity {z=-75}
   }
   n20 = RigidBody!(
    ComponentPose=@n19
    Alpha=Get front end() -> (front end:double)
    ExternalForce
    ComponentLinearAccScale:Vector="0.25,0.6,0.3"
    ComponentLinearVelScale:Vector="0.15,0.45,0.15"
   ) -> (Pose) {
    bClampLinearTranslationLimitToRefPose=true
    bOverrideWorldGravity=true
    overridePhysicsAsset=&/F_MED_Stone_Viper_Front_End_Physics
    simulationSpace=BaseBoneSpace
    alphaBoolBlend {blendInTime=0.5, blendOption=Linear, blendOutTime=0.5}
    baseBoneRef {boneName=spine_05}
    componentAppliedLinearAccClamp=<5000 5000 5000>
    componentLinearAccScale=<0.25 0.6 0.3>
    componentLinearVelScale=<0.15 0.45 0.15>
    overrideWorldGravity {z=-980}
   }
   n21 = "Transform (Modify) Bone - Bone: dyn_passive_MSTR"(ComponentPose=@n20, Alpha=1, Translation="0,0,0", Rotation="0,0,0", Scale=Get HairScaleInterp() -> (HairScaleInterp:Vector)) {scaleMode=BMM_Replace, alphaBoolBlend {blendInTime=0.3, blendOption=Linear, blendOutTime=0.3}, boneToModify {boneName=dyn_passive_MSTR}}
   n22 = "Transform (Modify) Bone - Bone: dyn_active_MSTR"(ComponentPose=@n21, Alpha=1, Translation="0,0,0", Rotation="0,0,0", Scale=Get SnakeHairScaleInterp() -> (SnakeHairScaleInterp:Vector)) {scaleMode=BMM_Replace, alphaBoolBlend {blendInTime=0.3, blendOption=Linear, blendOutTime=0.3}, boneToModify {boneName=dyn_active_MSTR}}
   n23 = Get bIsSkyDiving()
   n24 = "Transform (Modify) Bone - Bone: dyn_snake_b_MSTR"(ComponentPose=@n22, bAlphaBoolEnabled=@n23, Translation="0,0,0", Rotation="0.000,5,15", Scale="1,1,1") {
    alphaInputType=Bool
    rotationMode=BMM_Additive
    alphaBoolBlend {blendOption=Linear}
    boneToModify {boneName=dyn_snake_b_MSTR}
    rotation {roll=15, yaw=5}
   }
   n25 = "Transform (Modify) Bone - Bone: dyn_snake_c_MSTR"(ComponentPose=@n24, bAlphaBoolEnabled=@n23, Translation="0,0,0", Rotation="0.000,2,20", Scale="1,1,1") {
    alphaInputType=Bool
    rotationMode=BMM_Additive
    alphaBoolBlend {blendOption=Linear}
    boneToModify {boneName=dyn_snake_c_MSTR}
    rotation {roll=20, yaw=2}
   }
   n26 = "Transform (Modify) Bone - Bone: dyn_snake_d_MSTR"(ComponentPose=@n25, bAlphaBoolEnabled=@n23, Translation="0,0,0", Rotation="0.000,0.000,20", Scale="1,1,1") {
    alphaInputType=Bool
    rotationMode=BMM_Additive
    alphaBoolBlend {blendOption=Linear}
    boneToModify {boneName=dyn_snake_d_MSTR}
    rotation {roll=20}
   }
   n27 = "Transform (Modify) Bone - Bone: dyn_snake_e_MSTR"(ComponentPose=@n26, bAlphaBoolEnabled=@n23, Translation="0,0,0", Rotation="0.000,-2,20", Scale="1,1,1") {
    alphaInputType=Bool
    rotationMode=BMM_Additive
    alphaBoolBlend {blendOption=Linear}
    boneToModify {boneName=dyn_snake_e_MSTR}
    rotation {roll=20, yaw=-2}
   }
   n28 = "Transform (Modify) Bone - Bone: dyn_snake_f_MSTR"(ComponentPose=@n27, bAlphaBoolEnabled=@n23, Translation="0,0,0", Rotation="0.000,-5,10", Scale="1,1,1") {
    alphaInputType=Bool
    rotationMode=BMM_Additive
    alphaBoolBlend {blendOption=Linear}
    boneToModify {boneName=dyn_snake_f_MSTR}
    rotation {roll=10, yaw=-5}
   }
   n29 = "Transform (Modify) Bone - Bone: dyn_snake_a_MSTR"!(ComponentPose=@n28, bAlphaBoolEnabled:bool=Get bIsSkydiveFloating(), Translation="0,0,0", Rotation="0,20,-15", Scale="1,1,1") -> (Pose) {
    alphaInputType=Bool
    rotationMode=BMM_Additive
    alphaBoolBlend {blendOption=Linear}
    boneToModify {boneName=dyn_snake_a_MSTR}
    rotation {roll=-15, yaw=20}
   }
   Output Pose(Result=Component To Local(ComponentPose=@n29)) {layerGroup=None}
   "Sequence Evaluator (by time) 'Stone_Viper_Loop'"!(ExplicitTime:float=Get ScaleAnimExplicitTime() -> (ScaleAnimExplicitTime:double), StartPosition:float=0) -> (Pose:Pose Link Structure) {
    is=title AnimGraphNode_SequenceEvaluator
    sequence=&Stone_Viper_Loop
   }
   Use cached pose 'animpose'()
   NOT Boolean(A=Get bIsTargeting())
  ]}
 Locomotion {
  entry=Start
  path=AnimGraph.AnimGraphNode_StateMachine_0.Locomotion
  ids=id
  nodes [
   AnimStateNode_0 = Start() {is=title AnimStateNode}
   AnimStateNode_1 = Loop() {is=title AnimStateNode}
   AnimStateTransitionNode_0 = AnimStateTransitionNode() {bAutomaticRuleBasedOnSequencePlayerInState=true, crossfadeDuration=1}
  ]
  links [{
    dst=Loop
    rule=AnimGraph.AnimGraphNode_StateMachine_0.Locomotion.AnimStateTransitionNode_0.Transition
    src=Start
    via=AnimStateTransitionNode_0
   }]
 }
 Start {
  path=AnimGraph.AnimGraphNode_StateMachine_0.Locomotion.AnimStateNode_0.Start
  nodes [
   n1 = Sequence Player 'Stone_Viper_IntoADS'() {bLoopAnimation=false, sequence=&/Stone_Viper_IntoADS}
   Output Animation Pose(Result=@n1)
  ]
 }
 Loop {
  path=AnimGraph.AnimGraphNode_StateMachine_0.Locomotion.AnimStateNode_1.Loop
  nodes [Output Animation Pose(Result=Sequence Player 'Stone_Viper_Loop'() {sequence=&Stone_Viper_Loop})]
 }
 Transition {path=AnimGraph.AnimGraphNode_StateMachine_0.Locomotion.AnimStateTransitionNode_0.Transition, nodes [Result()]}
 EventGraph {nodes [
   Delay(Duration=0.1)
   > ResetDynamics(InTeleportType=TeleportPhysics)
   > ResetDynamics(InTeleportType=ResetPhysics)
   PrintString!(execute, InString=Huzzah, bPrintToScreen, bPrintToLog, TextColor, Duration, Key) -> (then)
   > Delay(Duration=0.5)
   > Set bIsReactivePreview_SnakesFromADS(bIsReactivePreview_SnakesFromADS:bool=true) -> (Output_Get:bool)
   Event BlueprintInitializeAnimation()
   > Set Body_Mesh(Body_Mesh=GetOwnerSkeletalMeshComponent())
   K2Node_Event_1 = CustomEvent_2() {is=title K2Node_CustomEvent}
   > n1 = Cast To FortPlayerPawn(Object=TryGetPawnOwner())
   > Set FortPlayerPawn(FortPlayerPawn:Fort Player Pawn Object Reference=@n1) -> (Output_Get:Fort Player Pawn Object Reference)
   Event OnCharacterCustomizationLoaded()
   > n2 = Sequence()
   n3 = CheckForBackbling()
   > Set BackBling_Alpha(BackBling_Alpha:double=SelectFloat(A=0, B=1, bPickA=@n3)) -> (Output_Get:double)
   > Set RigidsAcc(RigidsAcc:Vector=SelectVector(A="0.15,0.25,0.3", B="0.25,0.6,0.3", bPickA=@n3)) -> (Output_Get:Vector)
   > Set RigidsVel(RigidsVel:Vector=SelectVector(A="0.1,0.1,0.2", B="0.15,0.45,0.15", bPickA=@n3)) -> (Output_Get:Vector)
   n4 = Branch(
    execute=@n2.then_1
    Condition=CharacterItemDefinitionHasAnyMetaTag(Tags=Make Array("[0]"=(TagName="Cosmetics.MetaTags.ShapeShifter.SnakesOff")))
   )
   > Set disable_snakes(disable_snakes:bool=true) -> (Output_Get:bool)
   Set disable_snakes(execute=@n4.else, disable_snakes=false)
   K2Node_Event_5 = CustomEvent!() -> (
    OutputDelegate
    then
    ChannelAndVariant:"Cosmetic Variant Info Structure (by ref)"
    Preview:"Fort Cosmetic Reactive Preview (F CRP) Object Reference"
   ) {is=title K2Node_CustomEvent}
   > Gate Is Variant Reactive Towards!(execute:Exec, CosmeticVariantInfo:"Cosmetic Variant Info Structure (by ref)"=@K2Node_Event_5.ChannelAndVariant) -> (true:Exec) {is=title K2Node_MacroInstance}
   K2Node_Event_6 = CustomEvent_1!() -> (
    OutputDelegate
    then
    ChannelAndVariant:"Cosmetic Variant Info Structure (by ref)"
    Preview:"Fort Cosmetic Reactive Preview (F CRP) Object Reference"
   ) {is=title K2Node_CustomEvent}
   > Gate Is Variant Reactive Towards(CosmeticVariantInfo=@K2Node_Event_6.ChannelAndVariant)
   > Set bIsReactivePreview_SnakesFromADS(bIsReactivePreview_SnakesFromADS=false)
   n5 = Event BlueprintUpdateAnimation()
   > n6 = Sequence!(execute) -> (then_0, then_1, then_2:Exec)
   n7 = Get bIsInFrontEnd()
   n8 = NOT Boolean(A=@n7)
   n9 = Get bIsTargeting()
   > Set Rigids_Alpha(Rigids_Alpha:double=SelectFloat(A=0.6, B=0, bPickA=AND Boolean(A=@n9, B=@n8))) -> (Output_Get:double)
   > Set Hair_RBAN_Alpha(Hair_RBAN_Alpha:double=SelectFloat(A=0, B=0.6, bPickA=@n9)) -> (Output_Get:double)
   > Set front end(front end:double=SelectFloat(A=0.4, B=0, bPickA=@n7)) -> (Output_Get:double)
   > Set HairScale(HairScale:Vector=SelectVector(A="0.01,0.01,0.01", B="1,1,1", bPickA=Get bIsTargeting())) -> (Output_Get:Vector)
   > Set SnakeHairScale(SnakeHairScale:Vector=SelectVector(A="1,1,1", B="0.01,0.01,0.01", bPickA=Get bIsTargeting())) -> (Output_Get:Vector)
   > Set SnakeHairScaleInterp(SnakeHairScaleInterp:Vector=VInterpTo!(
     Current:Vector=Get SnakeHairScaleInterp()
     Target:Vector=Get SnakeHairScale() -> (SnakeHairScale:Vector)
     DeltaTime:float=@n5.DeltaTimeX
     InterpSpeed:float=20
    ) -> (ReturnValue:Vector)) -> (Output_Get:Vector)
   > Set HairScaleInterp(HairScaleInterp:Vector=VInterpTo(Current=Get HairScaleInterp(), Target=Get HairScale() -> (HairScale:Vector), DeltaTime=@n5.DeltaTimeX, InterpSpeed=20)) -> (Output_Get:Vector)
   > Set Hair_RBAN_AlphaInterp(Hair_RBAN_AlphaInterp:double=FInterpTo(Current=Get Hair_RBAN_AlphaInterp(), Target=Get Hair_RBAN_Alpha() -> (Hair_RBAN_Alpha:double), DeltaTime=@n5.DeltaTimeX, InterpSpeed=20)) -> (Output_Get:double)
   > Set ScaleAnimExplicitTime(ScaleAnimExplicitTime:double=FInterpTo(Current=Get ScaleAnimExplicitTime(), Target=Get Rigids_Alpha(), DeltaTime=@n5.DeltaTimeX, InterpSpeed=0)) -> (Output_Get:double)
   n10 = Gravity Override State Function DEPRECATED(
    execute=@n6.then_1
    CharacterPartAnimInstance=Self-Reference()
    GameTimeInSeconds=GetGameTimeInSeconds()
    GravityOverrideParams=Get Gravity Override Params()
   )
   > Set GravityOverride(GravityOverride=@n10)
   n11 = Get SkydivingExternalForceInput()
   SelectVector(
    A=vector * float(
     A=Break Fort Anim Input Skydiving External Force(FortAnimInput_SkydivingExternalForce=@n11) -> (
      bUseSkydivingVectorForce=~
      SkydivingVectorForce=~
      HeadToPelvisDirection=~
      FloatingMultiplier=~
      FloatingAdditive=~
      DivingMultiplier=~
      DivingAdditive=~
      ParachutingMultiplier=~
      ParachutingAdditive=~
     ).SkydivingVectorForce
     B=0.2
    )
    B="0, 0,-600"
    bPickA=Get bIsSkyDiving()
   )
   SelectFloat(A=0.43, B=0, bPickA=Get IsTargeting() -> (IsTargeting:bool))
   Event OnBeginTargeting() {is=title K2Node_Event}
  ]}
 Gate_IsVariantReactiveTowards {nodes [
   n1 = Inputs!() -> (execute, CosmeticVariantInfo:"Cosmetic Variant Info Structure (by ref)")
   n2 = Break Cosmetic Variant Info!(CosmeticVariantInfo:"Cosmetic Variant Info Structure (by ref)"=@n1) -> (VariantChannelTag:Gameplay Tag Structure, ActiveVariantTag:Gameplay Tag Structure) {is=title K2Node_BreakStruct}
   > Branch(Condition=AND Boolean(
     A="Equal (GameplayTag)"!(A:Gameplay Tag Structure=@n2, B:Gameplay Tag Structure=(TagName="Cosmetics.Variant.Channel.Parts")) -> (ReturnValue:bool) {is=title K2Node_CallFunction}
     B="Equal (GameplayTag)"(A=@n2.ActiveVariantTag, B=(TagName="Cosmetics.Variant.Property.Stage1"))
    ))
   > Outputs!(true:Exec)
  ]}
]
variables [
 BackBling_Alpha {type=number}
 bIsReactivePreview_SnakesFromADS {type=boolean}
 Body_Mesh {type=SkeletalMeshComponent}
 disable_snakes {type=boolean}
 FortPlayerPawn {type=FortPlayerPawn}
 front end {type=number}
 Gravity Override Params {
  type=GravityOverrideParamsStruct
  default {
   emoteOrMelee {gravityOverride=<0 0 -600>}
   front_End {gravityOverride=<0 0 -600>}
   onGround_Moving {gravityOverride=<0 0 -600>}
   onGround_Standing {gravityOverride=<0 0 -600>}
   skydive_DivingDown {windAmplitude=<50 -250 100>, windFrequency=<0.75 3.5 4>}
   skydive_FloatingDown {gravityOverride=<0 0 0>, windAmplitude=<100 -250 200>, windFrequency=<0.75 3.5 4>, windOffset=<0 -300 200>}
   skydive_GoingUp {gravityOverride=<0 0 -200>, windAmplitude=<100 -250 200>, windFrequency=<0.75 3.5 4>, windOffset=<0 -300 200>}
   skydive_Parachuting {gravityOverride=<0 0 -600>, windAmplitude=<100 -250 200>, windFrequency=<0.75 3.5 4>, windOffset=<0 -300 200>}
  }
 }
 GravityOverride {type=Vector}
 Hair_RBAN_Alpha {type=number}
 Hair_RBAN_AlphaInterp {type=number}
 HairScale {type=Vector}
 HairScaleInterp {type=Vector}
 IsTargeting {type=boolean}
 isWearingBackbling {type=boolean}
 Rigids_Alpha {type=number}
 RigidsAcc {type=Vector}
 RigidsVel {type=Vector}
 ScaleAnimExplicitTime {type=number}
 SnakeHairScale {type=Vector}
 SnakeHairScaleInterp {type=Vector}
 Target {type=Character, default=null}
 upperarm_l_rot {type=Rotator}
 upperarm_l_tran {type=Vector}
 upperarm_r_rot {type=Rotator}
 upperarm_r_tran {type=Vector}
]
